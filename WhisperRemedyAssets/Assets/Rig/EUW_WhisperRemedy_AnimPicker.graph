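# Editor Utility animation picker for the four-legged WhisperRemedy Control Rig, built on the AnimInEdTools picker base.
# Canvas buttons represent rig controls; PreConstruct binds them by suffix-matching widget names against the companion rig's control names.
# The companion rig comes from the binding proxy cached when the widget loads.
# Checkboxes key "Draw Capsule" channels on the current sequence's rigs and toggle the global p.AnimDynamics cvar, where checked means off.
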EditorUtilityWidgetBlueprint /WhisperRemedyAssets/Assets/Rig/EUW_WhisperRemedy_AnimPicker
graphs [
 EventGraph {nodes [
   n1 = "On Check State Changed (ToggleDynamicscheckbox)"!() -> (OutputDelegate, then, bIsChecked:bool) {is=title K2Node_ComponentBoundEvent}
   > n2 = Branch(Condition=@n1.bIsChecked)
   > ExecuteConsoleCommand(Command=p.AnimDynamics 0)
   ExecuteConsoleCommand(execute=@n2.else, Command=p.AnimDynamics 1)
   n3 = "On Check State Changed (ToggleCollisioncheckbox)"!() -> (OutputDelegate, then, bIsChecked:bool) {is=title K2Node_ComponentBoundEvent}
   > n4 = GetCurrentLevelSequence!(execute) -> (then, ReturnValue:Level Sequence Object Reference)
   > n5 = GetControlRigs!(execute, LevelSequence:Level Sequence Object Reference=@n4) -> (then, ReturnValue:Array of Control Rig Sequencer Binding Proxy Structures)
   > n6 = For Each Loop(Array:Array of Control Rig Sequencer Binding Proxy Structures=@n5) -> (Array Element:Control Rig Sequencer Binding Proxy Structure)
   n7 = Break Control Rig Sequencer Binding Proxy!(ControlRigSequencerBindingProxy:"Control Rig Sequencer Binding Proxy Structure (by ref)"=@n6) -> (
    Proxy:Movie Scene Binding Proxy Structure
    ControlRig:Control Rig Object Reference
    Track:Movie Scene Control Rig Parameter Track Object Reference
   ) {is=title K2Node_BreakStruct}
   n8 = GetHierarchy!(self:Control Rig Object Reference=@n7.ControlRig) -> (ReturnValue:Rig Hierarchy Object Reference)
   > n9 = For Each Loop(Array:Array of Rig Element Key Structures=Get Controls!(self:Rig Hierarchy Object Reference=@n8, bTraverse:bool=true) -> (ReturnValue:Array of Rig Element Key Structures) {is=title K2Node_CallFunction}) -> (Array Element:Rig Element Key Structure)
   > n10 = GetController!(execute, self:Rig Hierarchy Object Reference=@n8, bCreateIfNeeded:bool=false) -> (then, ReturnValue:Rig Hierarchy Controller Object Reference)
   n11 = Break Rig Control Settings!(RigControlSettings:"Rig Control Settings Structure (by ref)"=GetControlSettings!(self:Rig Hierarchy Controller Object Reference=@n10, InKey:Rig Element Key Structure=@n9) -> (ReturnValue:Rig Control Settings Structure)) -> (
    AnimationType:ERigControlAnimationType Enum
    ControlType:ERigControlType Enum
    DisplayName:Name
    PrimaryAxis:ERigControlAxis Enum
    LimitEnabled:Array of Rig Control Limit Enabled Structures
    bDrawLimits:bool=~
    MinimumValue:Rig Control Value Structure
    MaximumValue:Rig Control Value Structure
    bShapeVisible:bool=~
    ShapeVisibility:ERigControlVisibility Enum
    ShapeName:Name
    ShapeColor:Linear Color Structure
    bIsTransientControl:bool=~
    ControlEnum:Enum Object Reference
    Customization:Rig Control Element Customization Structure
    DrivenControls:Array of Rig Element Key Structures
    bGroupWithParentControl:bool=~
    bRestrictSpaceSwitching:bool=~
    FilteredChannels:Array of ERigControlTransformChannel Enums
    PreferredRotationOrder:EEulerRotationOrder Enum
    bUsePreferredRotationOrder:bool=~
    ShapeTransform:Transform
   ) {is=title K2Node_BreakStruct}
   n12 = "To String (Name)"(InName=@n11.DisplayName)
   n13 = "Equal (Enum)"(A:ERigControlType Enum=@n11.ControlType, B:ERigControlType Enum=Bool)
   n14 = "Equal (Enum)"(A:ERigControlAnimationType Enum=@n11, B:ERigControlAnimationType Enum=AnimationChannel)
   > Branch(Condition=AND Boolean(A=AND Boolean(A=@n14, B=@n13), B="Equal Exactly (String)"(A=@n12, B=Draw Capsule)))
   > SetLocalControlRigBool!(
    execute
    LevelSequence:Level Sequence Object Reference=@n4
    ControlRig:Control Rig Object Reference=@n7.ControlRig
    ControlName:Name=Break Rig Element Key!(RigElementKey:"Rig Element Key Structure (by ref)"=@n9) -> (Type:ERigElementType Enum, Name:Name) {is=title K2Node_BreakStruct}.Name
    Frame:Frame Number Structure
    Value:bool=@n3.bIsChecked
    TimeUnit:EMovieSceneTimeUnit Enum=DisplayRate
    bSetKey:bool=true
   ) -> (then)
   n15 = Event Widget has loaded!() -> (
    OutputDelegate
    then
    Control Rig Proxy Binding:Control Rig Sequencer Binding Proxy Structure
    Animatable Creature:Animatable Creature Object Reference
    Anim Picker EUW:Animation Picker Object Reference
   ) {is=title K2Node_Event}
   > Parent: Widget has loaded!(
    execute
    Control Rig Proxy Binding:Control Rig Sequencer Binding Proxy Structure=@n15.Control Rig Proxy Binding
    Animatable Creature:Animatable Creature Object Reference
    Anim Picker EUW:Animation Picker Object Reference
   ) -> (then) {is=title K2Node_CallParentFunction}
   > Set Control Rig Proxy Binding_0(Control Rig Proxy Binding_0:Control Rig Sequencer Binding Proxy Structure=@n15.Control Rig Proxy Binding) -> (Output_Get:Control Rig Sequencer Binding Proxy Structure)
   Event Construct()
   n16 = Break Control Rig Sequencer Binding Proxy(ControlRigSequencerBindingProxy=Get Control Rig Proxy Binding_0() -> (Control Rig Proxy Binding_0:Control Rig Sequencer Binding Proxy Structure))
   > Is Valid(InputObject=@n16.ControlRig)
   > Set Companion Control Rig(Companion Control Rig:Control Rig Object Reference=@n16.ControlRig) -> (Output_Get:Control Rig Object Reference)
   Event PreConstruct()
   > n17 = For Each Loop(Array:Array of Widget Object References=GetAllChildren(self=Get CanvasPanel() -> (CanvasPanel:Canvas Panel Object Reference))) -> (Array Element:Widget Object Reference)
   > n18 = Cast To EUW_Picker_Button!(execute, Object=@n17) -> (then, CastFailed, AsEUW Picker Button:EUW Picker Button Object Reference) {is=title K2Node_DynamicCast}
   n19 = Get Companion Control Rig() -> (Companion Control Rig:Control Rig Object Reference)
   > n20 = For Each Loop with Break!(Exec:Exec, Array:Array of Rig Element Key Structures=Get Controls(self=GetHierarchy(self=@n19), bTraverse=true), Break:Exec=@n24.then) -> (LoopBody:Exec, Array Element_Type:ERigElementType Enum, Array Element_Name:Name, Array Index:int, Completed:Exec)
   n21 = "To String (Name)"(InName=@n20.Array Element_Name)
   n22 = GetObjectName(Object=@n18)
   n23 = Replace(SourceString=Replace(SourceString=@n22, From=-, To=/), From=_, To=" ")
   > Branch(Condition=OR Boolean(A=EndsWith(SourceString=@n21, InSuffix=@n22), B=EndsWith(SourceString=@n21, InSuffix=@n23)))
   > n24 = Set Controls for Primary Control Rig!(
    execute
    self:EUW Picker Element Object Reference=@n18
    Control Names:Array of Names=Make Array("[0]":Name=@n20.Array Element_Name) -> (Array:Array of Names)
   ) -> (then) {is=title K2Node_CallFunction}
   Event Update From Scene() {is=title K2Node_Event}
   > Parent: Update From Scene() {is=title K2Node_CallParentFunction}
   > GetFocusedLevelSequence!(execute) -> (then, ReturnValue:Level Sequence Object Reference)
   n25 = For Each Loop(Array=GetAllChildren(self=Get CanvasPanel()))
   > n26 = Cast To EUW_Picker_Button(Object=@n25)
   n27 = GetObjectName(Object=@n26)
   > Set Control for Primary Control Rig!(
    execute
    self:EUW Picker Element Object Reference=@n26
    Control Names:String=Replace(SourceString=Replace(SourceString=@n27, From=-, To=:), From=_, To=" ")
   ) -> (then) {is=title K2Node_CallFunction}
  ]}
 Get Primary Panel {nodes [
   Get Primary Panel() {is=title K2Node_FunctionEntry}
   > Return Node!(execute, Panel:Panel Widget Object Reference=Get CanvasPanel())
  ]}
]
variables [
 Canvas Transform {type=WidgetTransform, default {angle=0, scale=<1 1>, shear=<0 0>, translation=<0 0>}}
 Companion Control Rig {type=ControlRig, default=null}
 Control Rig Proxy Binding_0 {
  type=ControlRigSequencerBindingProxy
  default {controlRig=null, track=null, proxy {bindingId=#1, sequence=null}}
 }
 Right Click Start {type=Vector2D, default=<0 0>}
]
widgets {
 info {
  parentClass=/AnimInEdTools/AnimationPicker/EUW_Asset_Picker_Base.EUW_Asset_Picker_Base_C
  rootWidgetClass=&CanvasPanel
  widgetCount=79
 }
 widgets [
  CanvasPanel {widgetClassPath=&CanvasPanel, bIsVariable=true, widget=CanvasPanel}
  Root_Main {
   widgetClassPath=/AnimInEdTools/AnimationPicker/EUW_Picker_Button
   bIsVariable=true
   parent=CanvasPanel
   widget=Root_Main
   properties {draw Style=Image, toolTipText=Root Main Ctrl}
   slot {class=CanvasPanelSlot, properties {layoutData {alignment=<0 0>, anchors {maximum=<0 0>, minimum=<0 0>}, offsets=<ltrb 140 472 240 213>}}}
  }
  Root_Offset {
   widgetClassPath=&EUW_Picker_Button
   bIsVariable=true
   parent=CanvasPanel
   widget=Root_Offset
   properties {draw Style=Image, toolTipText=Root Offset Ctrl, normal Tint {colorUseRule=UseColor_Specified, specifiedColor=<rgba 1 0.4515 0 1>}}
   slot {class=CanvasPanelSlot, properties {layoutData {alignment=<0 0>, anchors {maximum=<0 0>, minimum=<0 0>}, offsets=<ltrb 184 508 146 121>}}}
  }
  Body {
   widgetClassPath=&EUW_Picker_Button
   bIsVariable=true
   parent=CanvasPanel
   widget=Body
   properties {
    draw Style=Square
    toolTipText=Body Ctrl
    label Color {colorUseRule=UseColor_Specified, specifiedColor=<rgba 0 0 0 1>}
    normal Tint {colorUseRule=UseColor_Specified, specifiedColor=<rgba 1 0.4212 0 1>}
    renderTransform {angle=0, scale=<1 1>, shear=<0 0>, translation=<-79.0791 54.054>}
   }
   slot {
    class=CanvasPanelSlot
    properties {zOrder=1, layoutData {alignment=<0 0>, anchors {maximum=<0 0>, minimum=<0 0>}, offsets=<ltrb 300 472 73.2962 35.8087>}}
   }
  }
  Body_Offset {
   widgetClassPath=&EUW_Picker_Button
   bIsVariable=true
   parent=CanvasPanel
   widget=Body_Offset
   properties {
    draw Style=Square
    toolTipText=Body Offset Ctrl
    label Color {colorUseRule=UseColor_Specified, specifiedColor=<rgba 0 0 0 1>}
    normal Tint {colorUseRule=UseColor_Specified, specifiedColor=<rgba 1 0.4212 0 0.7913>}
    renderTransform {angle=0, scale=<1 1>, shear=<0 0>, translation=<-79.0791 54.054>}
   }
   slot {
    class=CanvasPanelSlot
    properties {zOrder=1, layoutData {alignment=<0 0>, anchors {maximum=<0 0>, minimum=<0 0>}, offsets=<ltrb 300 516 73.2962 34.9507>}}
   }
  }
  R_Fr_Heel_Roll {
   widgetClassPath=&EUW_Picker_Button
   bIsVariable=true
   parent=CanvasPanel
   widget=R_Fr_Heel_Roll
   properties {draw Style=Circle, toolTipText=R Fr Heel Roll, normal Tint {colorUseRule=UseColor_Specified, specifiedColor=<rgba 1 0.0137 0 1>}}
   slot {class=CanvasPanelSlot, properties {layoutData {alignment=<0 0>, anchors {maximum=<0 0>, minimum=<0 0>}, offsets=<ltrb 816 1020 50 50>}}}
  }
  R_Bk_Heel_Roll {
   widgetClassPath=&EUW_Picker_Button
   bIsVariable=true
   parent=CanvasPanel
   widget=R_Bk_Heel_Roll
   properties {draw Style=Circle, toolTipText=R Bk Heel Roll, normal Tint {colorUseRule=UseColor_Specified, specifiedColor=<rgba 1 0.0137 0 1>}}
   slot {class=CanvasPanelSlot, properties {layoutData {alignment=<0 0>, anchors {maximum=<0 0>, minimum=<0 0>}, offsets=<ltrb 1405.97 1008 50 50>}}}
  }
  L_Fr_Heel_Roll {
   widgetClassPath=&EUW_Picker_Button
   bIsVariable=true
   parent=CanvasPanel
   widget=L_Fr_Heel_Roll
   properties {draw Style=Circle, toolTipText=L Fr Heel Roll, normal Tint {colorUseRule=UseColor_Specified, specifiedColor=<rgba 0 0.0735 1 1>}}
   slot {class=CanvasPanelSlot, properties {layoutData {alignment=<0 0>, anchors {maximum=<0 0>, minimum=<0 0>}, offsets=<ltrb 600 984 50 50>}}}
  }
  L_Bk_Heel_Roll {
   widgetClassPath=&EUW_Picker_Button
   bIsVariable=true
   parent=CanvasPanel
   widget=L_Bk_Heel_Roll
   properties {draw Style=Circle, toolTipText=L Bk Heel Roll, normal Tint {colorUseRule=UseColor_Specified, specifiedColor=<rgba 0 0.0735 1 1>}}
   slot {class=CanvasPanelSlot, properties {layoutData {alignment=<0 0>, anchors {maximum=<0 0>, minimum=<0 0>}, offsets=<ltrb 1188 976 50 50>}}}
  }
  L_Fr_Toe_Roll {
   widgetClassPath=&EUW_Picker_Button
   bIsVariable=true
   parent=CanvasPanel
   widget=L_Fr_Toe_Roll
   properties {draw Style=Circle, toolTipText=L Fr Toe Roll, normal Tint {colorUseRule=UseColor_Specified, specifiedColor=<rgba 0 0.0735 1 1>}}
   slot {class=CanvasPanelSlot, properties {layoutData {alignment=<0 0>, anchors {maximum=<0 0>, minimum=<0 0>}, offsets=<ltrb 504 980 50 50>}}}
  }
  L_Bk_Toe_Roll {
   widgetClassPath=&EUW_Picker_Button
   bIsVariable=true
   parent=CanvasPanel
   widget=L_Bk_Toe_Roll
   properties {draw Style=Circle, toolTipText=L Bk Toe Roll, normal Tint {colorUseRule=UseColor_Specified, specifiedColor=<rgba 0 0.0735 1 1>}}
   slot {class=CanvasPanelSlot, properties {layoutData {alignment=<0 0>, anchors {maximum=<0 0>, minimum=<0 0>}, offsets=<ltrb 1084 976 50 50>}}}
  }
  R_Fr_Toe_Roll {
   widgetClassPath=&EUW_Picker_Button
   bIsVariable=true
   parent=CanvasPanel
   widget=R_Fr_Toe_Roll
   properties {draw Style=Circle, toolTipText=R Fr Toe Roll, normal Tint {colorUseRule=UseColor_Specified, specifiedColor=<rgba 1 0.0137 0 1>}}
   slot {class=CanvasPanelSlot, properties {layoutData {alignment=<0 0>, anchors {maximum=<0 0>, minimum=<0 0>}, offsets=<ltrb 700 1020.997 50 50>}}}
  }
  R_Bk_Toe_Roll {
   widgetClassPath=&EUW_Picker_Button
   bIsVariable=true
   parent=CanvasPanel
   widget=R_Bk_Toe_Roll
   properties {draw Style=Circle, toolTipText=R Bk Toe Roll, normal Tint {colorUseRule=UseColor_Specified, specifiedColor=<rgba 1 0.0137 0 1>}}
   slot {
    class=CanvasPanelSlot
    properties {layoutData {alignment=<0 0>, anchors {maximum=<0 0>, minimum=<0 0>}, offsets=<ltrb 1292 1011.9999 50 50>}}
   }
  }
  R_Ball_Rotate {
   widgetClassPath=&EUW_Picker_Button
   bIsVariable=true
   parent=CanvasPanel
   widget=R_Ball_Rotate
   properties {toolTipText=R Ball Rotate, normal Tint {colorUseRule=UseColor_Specified, specifiedColor=<rgba 1 110.7462 0 1>}}
   slot {
    class=CanvasPanelSlot
    properties {layoutData {alignment=<0 0>, anchors {maximum=<0 0>, minimum=<0 0>}, offsets=<ltrb 803.9999 808 26.8102 114.2643>}}
   }
  }
  ToggleCollisioncheckbox {
   widgetClassPath=&EditorUtilityCheckBox
   bIsVariable=true
   parent=CanvasPanel
   widget=ToggleCollisioncheckbox
   slot {
    class=CanvasPanelSlot
    properties {bAutoSize=true, layoutData {alignment=<0 0>, anchors {maximum=<0 0>, minimum=<0 0>}, offsets=<ltrb 120 712 386.1749 41.716>}}
   }
  }
  TextBlock_1 {
   widgetClassPath=&TextBlock
   parent=ToggleCollisioncheckbox
   widget=TextBlock_1
   properties {
    text=Toggle Collision Capsule
    font {
     bForceMonospaced=false
     bMaterialIsStencil=false
     fontMaterial=null
     fontObject=&Roboto
     letterSpacing=0
     monospacedWidth=1
     size=18.75
     skewAmount=0
     typefaceFontName=Bold
     outlineSettings {
      bApplyOutlineToDropShadows=false
      bMiteredCorners=false
      bSeparateFillAlpha=false
      outlineMaterial=null
      outlineSize=0
      outlineColor=<rgba 0 0 0 1>
     }
    }
   }
   slot {class=PanelSlot}
  }
  ToggleDynamicscheckbox {
   widgetClassPath=&EditorUtilityCheckBox
   bIsVariable=true
   parent=CanvasPanel
   widget=ToggleDynamicscheckbox
   properties {widgetStyle {
     checkBoxType=CheckBox
     backgroundHoveredImage {
      bIsDynamicallyLoaded=false
      drawAs=RoundedBox
      imageType=NoImage
      mirroring=NoMirror
      resourceName=None
      resourceObject=null
      tiling=NoTile
      imageSize=<18 18>
      margin=<ltrb 0 0 0 0>
      outlineSettings {
       bUseBrushTransparency=false
       roundingType=FixedRadius
       width=1
       color {colorUseRule=UseColor_ColorTable, specifiedColor=<rgba 0 0 0 0>}
       cornerRadii=<3 3 3 3>
      }
      tintColor {colorUseRule=UseColor_Specified, specifiedColor=<rgba 0.0048 0.0048 0.0048 1>}
      uVRegion {bIsValid=false, max=<0 0>, min=<0 0>}
     }
     backgroundImage {
      bIsDynamicallyLoaded=false
      drawAs=RoundedBox
      imageType=NoImage
      mirroring=NoMirror
      resourceName=None
      resourceObject=null
      tiling=NoTile
      imageSize=<18 18>
      margin=<ltrb 0 0 0 0>
      outlineSettings {
       bUseBrushTransparency=false
       roundingType=FixedRadius
       width=1
       color {colorUseRule=UseColor_ColorTable, specifiedColor=<rgba 0 0 0 0>}
       cornerRadii=<3 3 3 3>
      }
      tintColor {colorUseRule=UseColor_Specified, specifiedColor=<rgba 0.0048 0.0048 0.0048 1>}
      uVRegion {bIsValid=false, max=<0 0>, min=<0 0>}
     }
     backgroundPressedImage {
      bIsDynamicallyLoaded=false
      drawAs=RoundedBox
      imageType=NoImage
      mirroring=NoMirror
      resourceName=None
      resourceObject=null
      tiling=NoTile
      imageSize=<18 18>
      margin=<ltrb 0 0 0 0>
      outlineSettings {
       bUseBrushTransparency=false
       roundingType=FixedRadius
       width=1
       color {colorUseRule=UseColor_ColorTable, specifiedColor=<rgba 0 0 0 0>}
       cornerRadii=<3 3 3 3>
      }
      tintColor {colorUseRule=UseColor_Specified, specifiedColor=<rgba 0.0048 0.0048 0.0048 1>}
      uVRegion {bIsValid=false, max=<0 0>, min=<0 0>}
     }
     borderBackgroundColor {colorUseRule=UseColor_Specified, specifiedColor=<rgba 1 1 1 1>}
     checkedForeground {colorUseRule=UseColor_Specified, specifiedColor=<rgba 1 1 1 1>}
     checkedHoveredForeground {colorUseRule=UseColor_Specified, specifiedColor=<rgba 1 1 1 1>}
     checkedHoveredImage {
      bIsDynamicallyLoaded=false
      drawAs=Image
      imageType=Vector
      mirroring=NoMirror
      resourceName=../../../Engine/Content/Slate/Starship/CoreWidgets/CheckBox/check.svg
      resourceObject=null
      tiling=NoTile
      imageSize=<16 16>
      margin=<ltrb 0 0 0 0>
      outlineSettings {
       bUseBrushTransparency=false
       roundingType=HalfHeightRadius
       width=0
       color {colorUseRule=UseColor_Specified, specifiedColor=<rgba 0 0 0 0>}
       cornerRadii=<0 0 0 0>
      }
      tintColor {colorUseRule=UseColor_Specified, specifiedColor=<rgba 0.0044 0.2384 1 1>}
      uVRegion {bIsValid=false, max=<0 0>, min=<0 0>}
     }
     checkedImage {
      bIsDynamicallyLoaded=false
      drawAs=Image
      imageType=Vector
      mirroring=NoMirror
      resourceName=../../../Engine/Content/Slate/Starship/CoreWidgets/CheckBox/check.svg
      resourceObject=null
      tiling=NoTile
      imageSize=<16 16>
      margin=<ltrb 0 0 0 0>
      outlineSettings {
       bUseBrushTransparency=false
       roundingType=HalfHeightRadius
       width=0
       color {colorUseRule=UseColor_Specified, specifiedColor=<rgba 0 0 0 0>}
       cornerRadii=<0 0 0 0>
      }
      tintColor {colorUseRule=UseColor_Specified, specifiedColor=<rgba 0 0.162 0.7454 1>}
      uVRegion {bIsValid=false, max=<0 0>, min=<0 0>}
     }
     checkedPressedForeground {colorUseRule=UseColor_Specified, specifiedColor=<rgba 1 0.4835 0.3716 1>}
     checkedPressedImage {
      bIsDynamicallyLoaded=false
      drawAs=Image
      imageType=Vector
      mirroring=NoMirror
      resourceName=../../../Engine/Content/Slate/Starship/CoreWidgets/CheckBox/check.svg
      resourceObject=null
      tiling=NoTile
      imageSize=<16 16>
      margin=<ltrb 0 0 0 0>
      outlineSettings {
       bUseBrushTransparency=false
       roundingType=HalfHeightRadius
       width=0
       color {colorUseRule=UseColor_Specified, specifiedColor=<rgba 0 0 0 0>}
       cornerRadii=<0 0 0 0>
      }
      tintColor {colorUseRule=UseColor_Specified, specifiedColor=<rgba 0 0.0802 0.3515 1>}
      uVRegion {bIsValid=false, max=<0 0>, min=<0 0>}
     }
     checkedSlateSound {resourceObject=null}
     foregroundColor {colorUseRule=UseColor_Specified, specifiedColor=<rgba 1 0.0253 0 1>}
     hoveredForeground {colorUseRule=UseColor_Specified, specifiedColor=<rgba 1 1 1 1>}
     hoveredSlateSound {resourceObject=null}
     padding=<ltrb 2 0 0 0>
     pressedForeground {colorUseRule=UseColor_Specified, specifiedColor=<rgba 1 1 1 1>}
     uncheckedHoveredImage {
      bIsDynamicallyLoaded=false
      drawAs=NoDrawType
      imageType=NoImage
      mirroring=NoMirror
      resourceName=None
      resourceObject=null
      tiling=NoTile
      imageSize=<0 0>
      margin=<ltrb 0 0 0 0>
      outlineSettings {
       bUseBrushTransparency=false
       roundingType=HalfHeightRadius
       width=0
       color {colorUseRule=UseColor_Specified, specifiedColor=<rgba 0 0 0 0>}
       cornerRadii=<0 0 0 0>
      }
      tintColor {colorUseRule=UseColor_Specified, specifiedColor=<rgba 1 1 1 1>}
      uVRegion {bIsValid=false, max=<0 0>, min=<0 0>}
     }
     uncheckedImage {
      bIsDynamicallyLoaded=false
      drawAs=NoDrawType
      imageType=NoImage
      mirroring=NoMirror
      resourceName=None
      resourceObject=null
      tiling=NoTile
      imageSize=<0 0>
      margin=<ltrb 0 0 0 0>
      outlineSettings {
       bUseBrushTransparency=false
       roundingType=HalfHeightRadius
       width=0
       color {colorUseRule=UseColor_Specified, specifiedColor=<rgba 0 0 0 0>}
       cornerRadii=<0 0 0 0>
      }
      tintColor {colorUseRule=UseColor_Specified, specifiedColor=<rgba 1 1 1 1>}
      uVRegion {bIsValid=false, max=<0 0>, min=<0 0>}
     }
     uncheckedPressedImage {
      bIsDynamicallyLoaded=false
      drawAs=NoDrawType
      imageType=NoImage
      mirroring=NoMirror
      resourceName=None
      resourceObject=null
      tiling=NoTile
      imageSize=<0 0>
      margin=<ltrb 0 0 0 0>
      outlineSettings {
       bUseBrushTransparency=false
       roundingType=HalfHeightRadius
       width=0
       color {colorUseRule=UseColor_Specified, specifiedColor=<rgba 0 0 0 0>}
       cornerRadii=<0 0 0 0>
      }
      tintColor {colorUseRule=UseColor_Specified, specifiedColor=<rgba 1 1 1 1>}
      uVRegion {bIsValid=false, max=<0 0>, min=<0 0>}
     }
     uncheckedSlateSound {resourceObject=null}
     undeterminedForeground {colorUseRule=UseColor_Specified, specifiedColor=<rgba 1 1 1 1>}
     undeterminedHoveredImage {
      bIsDynamicallyLoaded=false
      drawAs=Image
      imageType=Vector
      mirroring=NoMirror
      resourceName=../../../Engine/Content/Slate/Starship/CoreWidgets/CheckBox/indeterminate.svg
      resourceObject=null
      tiling=NoTile
      imageSize=<16 16>
      margin=<ltrb 0 0 0 0>
      outlineSettings {
       bUseBrushTransparency=false
       roundingType=HalfHeightRadius
       width=0
       color {colorUseRule=UseColor_Specified, specifiedColor=<rgba 0 0 0 0>}
       cornerRadii=<0 0 0 0>
      }
      tintColor {colorUseRule=UseColor_Specified, specifiedColor=<rgba 0.0044 0.2384 1 1>}
      uVRegion {bIsValid=false, max=<0 0>, min=<0 0>}
     }
     undeterminedImage {
      bIsDynamicallyLoaded=false
      drawAs=Image
      imageType=Vector
      mirroring=NoMirror
      resourceName=../../../Engine/Content/Slate/Starship/CoreWidgets/CheckBox/indeterminate.svg
      resourceObject=null
      tiling=NoTile
      imageSize=<16 16>
      margin=<ltrb 0 0 0 0>
      outlineSettings {
       bUseBrushTransparency=false
       roundingType=HalfHeightRadius
       width=0
       color {colorUseRule=UseColor_Specified, specifiedColor=<rgba 0 0 0 0>}
       cornerRadii=<0 0 0 0>
      }
      tintColor {colorUseRule=UseColor_Specified, specifiedColor=<rgba 0 0.162 0.7454 1>}
      uVRegion {bIsValid=false, max=<0 0>, min=<0 0>}
     }
     undeterminedPressedImage {
      bIsDynamicallyLoaded=false
      drawAs=Image
      imageType=Vector
      mirroring=NoMirror
      resourceName=../../../Engine/Content/Slate/Starship/CoreWidgets/CheckBox/indeterminate.svg
      resourceObject=null
      tiling=NoTile
      imageSize=<16 16>
      margin=<ltrb 0 0 0 0>
      outlineSettings {
       bUseBrushTransparency=false
       roundingType=HalfHeightRadius
       width=0
       color {colorUseRule=UseColor_Specified, specifiedColor=<rgba 0 0 0 0>}
       cornerRadii=<0 0 0 0>
      }
      tintColor {colorUseRule=UseColor_Specified, specifiedColor=<rgba 0 0.0802 0.3515 1>}
      uVRegion {bIsValid=false, max=<0 0>, min=<0 0>}
     }
    }}
   slot {
    class=CanvasPanelSlot
    properties {bAutoSize=true, layoutData {alignment=<0 0>, anchors {maximum=<0 0>, minimum=<0 0>}, offsets=<ltrb 120 768 386.1749 41.716>}}
   }
  }
  TextBlock_4 {
   widgetClassPath=&TextBlock
   parent=ToggleDynamicscheckbox
   widget=TextBlock_4
   properties {
    text=Toggle Tail Dynamics
    font {
     bForceMonospaced=false
     bMaterialIsStencil=false
     fontMaterial=null
     fontObject=&Roboto
     letterSpacing=0
     monospacedWidth=1
     size=18.75
     skewAmount=0
     typefaceFontName=Bold
     outlineSettings {
      bApplyOutlineToDropShadows=false
      bMiteredCorners=false
      bSeparateFillAlpha=false
      outlineMaterial=null
      outlineSize=0
      outlineColor=<rgba 0 0 0 1>
     }
    }
   }
   slot {class=PanelSlot}
  }
  L_Hock_Rotate {
   widgetClassPath=&EUW_Picker_Button
   bIsVariable=true
   parent=CanvasPanel
   widget=L_Hock_Rotate
   properties {toolTipText=L HockRotate, normal Tint {colorUseRule=UseColor_Specified, specifiedColor=<rgba 1 110.7462 0 1>}}
   slot {
    class=CanvasPanelSlot
    properties {layoutData {alignment=<0 0>, anchors {maximum=<0 0>, minimum=<0 0>}, offsets=<ltrb 1216 808 25.3086 106.2563>}}
   }
  }
  R_Hock_Rotate {
   widgetClassPath=&EUW_Picker_Button
   bIsVariable=true
   parent=CanvasPanel
   widget=R_Hock_Rotate
   properties {toolTipText=R Hock Rotate, normal Tint {colorUseRule=UseColor_Specified, specifiedColor=<rgba 1 110.7462 0 1>}}
   slot {
    class=CanvasPanelSlot
    properties {layoutData {alignment=<0 0>, anchors {maximum=<0 0>, minimum=<0 0>}, offsets=<ltrb 1420.7267 824 33.5669 106.2563>}}
   }
  }
  L_Ball_Rotate {
   widgetClassPath=&EUW_Picker_Button
   bIsVariable=true
   parent=CanvasPanel
   widget=L_Ball_Rotate
   properties {toolTipText=L Ball Rotate, normal Tint {colorUseRule=UseColor_Specified, specifiedColor=<rgba 1 0.9365 0 1>}}
   slot {class=CanvasPanelSlot, properties {layoutData {alignment=<0 0>, anchors {maximum=<0 0>, minimum=<0 0>}, offsets=<ltrb 592 804 26.5165 106>}}}
  }
  Spine_Bk_Proxy {
   widgetClassPath=&EUW_Picker_Button
   bIsVariable=true
   parent=CanvasPanel
   widget=Spine_Bk_Proxy
   properties {draw Style=Circle, toolTipText=Spine Fwd Ctrl, normal Tint {colorUseRule=UseColor_Specified, specifiedColor=<rgba 1 0.5771 0 1>}}
   slot {
    class=CanvasPanelSlot
    properties {zOrder=1, layoutData {alignment=<0 0>, anchors {maximum=<0 0>, minimum=<0 0>}, offsets=<ltrb 1000 412 27.988 29.1892>}}
   }
  }
  Tail_Rotate {
   widgetClassPath=&EUW_Picker_Button
   bIsVariable=true
   parent=CanvasPanel
   widget=Tail_Rotate
   properties {draw Style=Circle, toolTipText=Tail Proxy Ctrl, normal Tint {colorUseRule=UseColor_Specified, specifiedColor=<rgba 0.101 1 0 1>}}
   slot {
    class=CanvasPanelSlot
    properties {zOrder=1, layoutData {alignment=<0 0>, anchors {maximum=<0 0>, minimum=<0 0>}, offsets=<ltrb 1240 356 31.006 33.008>}}
   }
  }
  Spine_Fwd_Proxy {
   widgetClassPath=&EUW_Picker_Button
   bIsVariable=true
   parent=CanvasPanel
   widget=Spine_Fwd_Proxy
   properties {draw Style=Circle, toolTipText=Spine BK Ctrl, normal Tint {colorUseRule=UseColor_Specified, specifiedColor=<rgba 1 0.5771 0 1>}}
   slot {
    class=CanvasPanelSlot
    properties {zOrder=1, layoutData {alignment=<0 0>, anchors {maximum=<0 0>, minimum=<0 0>}, offsets=<ltrb 848 408 30.3904 29.1892>}}
   }
  }
  Hip {
   widgetClassPath=&EUW_Picker_Button
   bIsVariable=true
   parent=CanvasPanel
   widget=Hip
   properties {draw Style=RoundedSquareLargeRadius, toolTipText=Hip Ctrl, normal Tint {colorUseRule=UseColor_Specified, specifiedColor=<rgba 1 0.5771 0 1>}}
   slot {
    class=CanvasPanelSlot
    properties {zOrder=1, layoutData {alignment=<0 0>, anchors {maximum=<0 0>, minimum=<0 0>}, offsets=<ltrb 1103.067 410.971 83.0831 224.8892>}}
   }
  }
  Spine_02 {
   widgetClassPath=&EUW_Picker_Button
   bIsVariable=true
   parent=CanvasPanel
   widget=Spine_02
   properties {toolTipText=Spine Mid Ctrl, normal Tint {colorUseRule=UseColor_Specified, specifiedColor=<rgba 0.9301 0.7529 0.0242 1>}}
   slot {
    class=CanvasPanelSlot
    properties {zOrder=1, layoutData {alignment=<0 0>, anchors {maximum=<0 0>, minimum=<0 0>}, offsets=<ltrb 888 440 100.796 171.6102>}}
   }
  }
  Chest {
   widgetClassPath=&EUW_Picker_Button
   bIsVariable=true
   parent=CanvasPanel
   widget=Chest
   properties {toolTipText=Chest Ctrl, normal Tint {colorUseRule=UseColor_Specified, specifiedColor=<rgba 1 0.5771 0 1>}}
   slot {
    class=CanvasPanelSlot
    properties {zOrder=1, layoutData {alignment=<0 0>, anchors {maximum=<0 0>, minimum=<0 0>}, offsets=<ltrb 648 401.962 175.2038 222.2008>}}
   }
  }
  Head_Main {
   widgetClassPath=&EUW_Picker_Button
   bIsVariable=true
   parent=CanvasPanel
   widget=Head_Main
   properties {draw Style=Circle, toolTipText=Head Ctrl, normal Tint {colorUseRule=UseColor_Specified, specifiedColor=<rgba 0.9301 0.7529 0.0242 1>}}
   slot {
    class=CanvasPanelSlot
    properties {zOrder=1, layoutData {alignment=<0 0>, anchors {maximum=<0 0>, minimum=<0 0>}, offsets=<ltrb 376 148 177.7606 177.2029>}}
   }
  }
  Neck_2 {
   widgetClassPath=&EUW_Picker_Button
   bIsVariable=true
   parent=CanvasPanel
   widget=Neck_2
   properties {draw Style=Circle, toolTipText=Neck 02 Control, normal Tint {colorUseRule=UseColor_Specified, specifiedColor=<rgba 0.9301 0.7529 0.0242 1>}}
   slot {
    class=CanvasPanelSlot
    properties {zOrder=1, layoutData {alignment=<0 0>, anchors {maximum=<0 0>, minimum=<0 0>}, offsets=<ltrb 464 348 121.7046 33.831>}}
   }
  }
  Neck_1 {
   widgetClassPath=&EUW_Picker_Button
   bIsVariable=true
   parent=CanvasPanel
   widget=Neck_1
   properties {draw Style=Circle, toolTipText=Neck 01 Control, normal Tint {colorUseRule=UseColor_Specified, specifiedColor=<rgba 0.9301 0.7529 0.0242 1>}}
   slot {
    class=CanvasPanelSlot
    properties {zOrder=1, layoutData {alignment=<0 0>, anchors {maximum=<0 0>, minimum=<0 0>}, offsets=<ltrb 484 388 151.7346 34.689>}}
   }
  }
  Neck_IK_02 {
   widgetClassPath=&EUW_Picker_Button
   bIsVariable=true
   parent=CanvasPanel
   widget=Neck_IK_02
   properties {draw Style=Circle, toolTipText=Neck IK Ctrl, normal Tint {colorUseRule=UseColor_Specified, specifiedColor=<rgba 1 0.8353 0 1>}}
   slot {
    class=CanvasPanelSlot
    properties {zOrder=1, layoutData {alignment=<0 0>, anchors {maximum=<0 0>, minimum=<0 0>}, offsets=<ltrb 420 384 32.8724 31.5916>}}
   }
  }
  Spine_01 {
   widgetClassPath=&EUW_Picker_Button
   bIsVariable=true
   parent=CanvasPanel
   widget=Spine_01
   properties {toolTipText=Spine Def 01 Ctrl, normal Tint {colorUseRule=UseColor_Specified, specifiedColor=<rgba 1 0.2607 0 1>}}
   slot {
    class=CanvasPanelSlot
    properties {zOrder=1, layoutData {alignment=<0 0>, anchors {maximum=<0 0>, minimum=<0 0>}, offsets=<ltrb 1057.009 440.937 35.9359 192.1421>}}
   }
  }
  Pelvis_FK {
   widgetClassPath=&EUW_Picker_Button
   bIsVariable=true
   parent=CanvasPanel
   widget=Pelvis_FK
   properties {toolTipText=Pelvis Ctrl, normal Tint {colorUseRule=UseColor_Specified, specifiedColor=<rgba 0 0.2917 0.0173 1>}}
   slot {
    class=CanvasPanelSlot
    properties {zOrder=1, layoutData {alignment=<0 0>, anchors {maximum=<0 0>, minimum=<0 0>}, offsets=<ltrb 1194.0541 442.9269 34.9349 183.1582>}}
   }
  }
  R_Fr_Ik {
   widgetClassPath=&EUW_Picker_Button
   bIsVariable=true
   parent=CanvasPanel
   widget=R_Fr_Ik
   properties {
    draw Style=Square
    toolTipText=R Fr IK Ctrl
    label Color {colorUseRule=UseColor_Specified, specifiedColor=<rgba 0 0 0 1>}
    normal Tint {colorUseRule=UseColor_Specified, specifiedColor=<rgba 1 0 0 1>}
   }
   slot {
    class=CanvasPanelSlot
    properties {zOrder=1, layoutData {alignment=<0 0>, anchors {maximum=<0 0>, minimum=<0 0>}, offsets=<ltrb 772 956 60.1602 58.8338>}}
   }
  }
  R_Fr_Ball {
   widgetClassPath=&EUW_Picker_Button
   bIsVariable=true
   parent=CanvasPanel
   widget=R_Fr_Ball
   properties {
    draw Style=Circle
    toolTipText=R Fr Ball Ctrl
    label Color {colorUseRule=UseColor_Specified, specifiedColor=<rgba 0 0 0 1>}
    normal Tint {colorUseRule=UseColor_Specified, specifiedColor=<rgba 0 1 0.005 1>}
   }
   slot {
    class=CanvasPanelSlot
    properties {zOrder=1, layoutData {alignment=<0 0>, anchors {maximum=<0 0>, minimum=<0 0>}, offsets=<ltrb 836 956 43.564 42.472>}}
   }
  }
  R_Fr_Toe {
   widgetClassPath=&EUW_Picker_Button
   bIsVariable=true
   parent=CanvasPanel
   widget=R_Fr_Toe
   properties {
    draw Style=Square
    toolTipText=R Fr Toe Ctrl
    label Color {colorUseRule=UseColor_Specified, specifiedColor=<rgba 0 0 0 1>}
    normal Tint {colorUseRule=UseColor_Specified, specifiedColor=<rgba 1 0.4212 0 1>}
   }
   slot {
    class=CanvasPanelSlot
    properties {zOrder=1, layoutData {alignment=<0 0>, anchors {maximum=<0 0>, minimum=<0 0>}, offsets=<ltrb 732 976 35.472 37.11>}}
   }
  }
  R_Fr_Ball_Pivot {
   widgetClassPath=&EUW_Picker_Button
   bIsVariable=true
   parent=CanvasPanel
   widget=R_Fr_Ball_Pivot
   properties {
    draw Style=Circle
    toolTipText=R Fr Ball Pivot Ctrl
    label Color {colorUseRule=UseColor_Specified, specifiedColor=<rgba 0 0 0 1>}
    normal Tint {colorUseRule=UseColor_Specified, specifiedColor=<rgba 0.0833 0.8226 1 1>}
   }
   slot {
    class=CanvasPanelSlot
    properties {zOrder=1, layoutData {alignment=<0 0>, anchors {maximum=<0 0>, minimum=<0 0>}, offsets=<ltrb 696 980 30.926 29.288>}}
   }
  }
  L_Fr_Ik {
   widgetClassPath=&EUW_Picker_Button
   bIsVariable=true
   parent=CanvasPanel
   widget=L_Fr_Ik
   properties {
    draw Style=Square
    toolTipText=L Fr IK Ctrl
    label Color {colorUseRule=UseColor_Specified, specifiedColor=<rgba 0 0 0 1>}
    normal Tint {colorUseRule=UseColor_Specified, specifiedColor=<rgba 0 0 1 1>}
   }
   slot {
    class=CanvasPanelSlot
    properties {zOrder=1, layoutData {alignment=<0 0>, anchors {maximum=<0 0>, minimum=<0 0>}, offsets=<ltrb 568 916 62.5626 58.8338>}}
   }
  }
  L_Fr_Ball {
   widgetClassPath=&EUW_Picker_Button
   bIsVariable=true
   parent=CanvasPanel
   widget=L_Fr_Ball
   properties {
    draw Style=Circle
    toolTipText=L Fr Ball Ctrl
    label Color {colorUseRule=UseColor_Specified, specifiedColor=<rgba 0 0 0 1>}
    normal Tint {colorUseRule=UseColor_Specified, specifiedColor=<rgba 0 1 0 1>}
   }
   slot {
    class=CanvasPanelSlot
    properties {zOrder=1, layoutData {alignment=<0 0>, anchors {maximum=<0 0>, minimum=<0 0>}, offsets=<ltrb 636 920 45.4872 43.6854>}}
   }
  }
  L_Fr_Toe {
   widgetClassPath=&EUW_Picker_Button
   bIsVariable=true
   parent=CanvasPanel
   widget=L_Fr_Toe
   properties {
    draw Style=Square
    toolTipText=L Fr Toe Ctrl
    label Color {colorUseRule=UseColor_Specified, specifiedColor=<rgba 0 0 0 1>}
    normal Tint {colorUseRule=UseColor_Specified, specifiedColor=<rgba 1 0.4212 0 1>}
   }
   slot {
    class=CanvasPanelSlot
    properties {zOrder=1, layoutData {alignment=<0 0>, anchors {maximum=<0 0>, minimum=<0 0>}, offsets=<ltrb 523.5796 933.7778 38.4204 40.2222>}}
   }
  }
  L_Fr_Ball_Pivot {
   widgetClassPath=&EUW_Picker_Button
   bIsVariable=true
   parent=CanvasPanel
   widget=L_Fr_Ball_Pivot
   properties {
    draw Style=Circle
    toolTipText=L Fr Ball Pivot Ctrl
    label Color {colorUseRule=UseColor_Specified, specifiedColor=<rgba 0 0 0 1>}
    normal Tint {colorUseRule=UseColor_Specified, specifiedColor=<rgba 0.0833 0.8226 1 1>}
   }
   slot {
    class=CanvasPanelSlot
    properties {zOrder=1, layoutData {alignment=<0 0>, anchors {maximum=<0 0>, minimum=<0 0>}, offsets=<ltrb 488 942.6653 30.2829 27.9473>}}
   }
  }
  R_Bk_Ik {
   widgetClassPath=&EUW_Picker_Button
   bIsVariable=true
   parent=CanvasPanel
   widget=R_Bk_Ik
   properties {
    draw Style=Square
    toolTipText=R Back IK Ctrl
    label Color {colorUseRule=UseColor_Specified, specifiedColor=<rgba 0 0 0 1>}
    normal Tint {colorUseRule=UseColor_Specified, specifiedColor=<rgba 1 0 0 1>}
   }
   slot {
    class=CanvasPanelSlot
    properties {zOrder=1, layoutData {alignment=<0 0>, anchors {maximum=<0 0>, minimum=<0 0>}, offsets=<ltrb 1364 944 58.959 58.8338>}}
   }
  }
  R_Bk_Ball {
   widgetClassPath=&EUW_Picker_Button
   bIsVariable=true
   parent=CanvasPanel
   widget=R_Bk_Ball
   properties {
    draw Style=Circle
    toolTipText=R Bk Ball Ctrl
    label Color {colorUseRule=UseColor_Specified, specifiedColor=<rgba 0 0 0 1>}
    normal Tint {colorUseRule=UseColor_Specified, specifiedColor=<rgba 0 1 0.005 1>}
   }
   slot {
    class=CanvasPanelSlot
    properties {zOrder=1, layoutData {alignment=<0 0>, anchors {maximum=<0 0>, minimum=<0 0>}, offsets=<ltrb 1428 944 41.016 42.017>}}
   }
  }
  R_Knee_PV {
   widgetClassPath=&EUW_Picker_Button
   bIsVariable=true
   parent=CanvasPanel
   widget=R_Knee_PV
   properties {
    draw Style=Circle
    toolTipText=R_Knee_PV
    label Color {colorUseRule=UseColor_Specified, specifiedColor=<rgba 0 0 0 1>}
    normal Tint {colorUseRule=UseColor_Specified, specifiedColor=<rgba 1 0 0 1>}
   }
   slot {
    class=CanvasPanelSlot
    properties {zOrder=1, layoutData {alignment=<0 0>, anchors {maximum=<0 0>, minimum=<0 0>}, offsets=<ltrb 1296 800 53.2733 56.2763>}}
   }
  }
  R_Elbow_PV {
   widgetClassPath=&EUW_Picker_Button
   bIsVariable=true
   parent=CanvasPanel
   widget=R_Elbow_PV
   properties {
    draw Style=Circle
    toolTipText=R_Elbow_PV
    label Color {colorUseRule=UseColor_Specified, specifiedColor=<rgba 0 0 0 1>}
    normal Tint {colorUseRule=UseColor_Specified, specifiedColor=<rgba 1 0 0 1>}
   }
   slot {
    class=CanvasPanelSlot
    properties {zOrder=1, layoutData {alignment=<0 0>, anchors {maximum=<0 0>, minimum=<0 0>}, offsets=<ltrb 855.9999 800 52.5225 49.5195>}}
   }
  }
  L_Elbow_PV {
   widgetClassPath=&EUW_Picker_Button
   bIsVariable=true
   parent=CanvasPanel
   widget=L_Elbow_PV
   properties {
    draw Style=Circle
    toolTipText=L _Elbow_PV
    label Color {colorUseRule=UseColor_Specified, specifiedColor=<rgba 0 0 0 1>}
    normal Tint {colorUseRule=UseColor_Specified, specifiedColor=<rgba 0 0.0735 1 1>}
   }
   slot {
    class=CanvasPanelSlot
    properties {zOrder=1, layoutData {alignment=<0 0>, anchors {maximum=<0 0>, minimum=<0 0>}, offsets=<ltrb 644 800 50.2703 52.5225>}}
   }
  }
  L_Ear_Root_FK {
   widgetClassPath=&EUW_Picker_Button
   bIsVariable=true
   parent=CanvasPanel
   widget=L_Ear_Root_FK
   properties {
    draw Style=Circle
    toolTipText=L_Ear_FK_01
    label Color {colorUseRule=UseColor_Specified, specifiedColor=<rgba 0 0 0 1>}
    normal Tint {colorUseRule=UseColor_Specified, specifiedColor=<rgba 0 0.0735 1 1>}
   }
   slot {
    class=CanvasPanelSlot
    properties {zOrder=1, layoutData {alignment=<0 0>, anchors {maximum=<0 0>, minimum=<0 0>}, offsets=<ltrb 572 140 30 30>}}
   }
  }
  R_Ear_FK_02 {
   widgetClassPath=&EUW_Picker_Button
   bIsVariable=true
   parent=CanvasPanel
   widget=R_Ear_FK_02
   properties {
    draw Style=Circle
    toolTipText=R_Ear_FK_02
    label Color {colorUseRule=UseColor_Specified, specifiedColor=<rgba 0 0 0 1>}
    normal Tint {colorUseRule=UseColor_Specified, specifiedColor=<rgba 1 0.0137 0 1>}
   }
   slot {
    class=CanvasPanelSlot
    properties {zOrder=1, layoutData {alignment=<0 0>, anchors {maximum=<0 0>, minimum=<0 0>}, offsets=<ltrb 636 52 60.888 30>}}
   }
  }
  R_Ear_FK_03 {
   widgetClassPath=&EUW_Picker_Button
   bIsVariable=true
   parent=CanvasPanel
   widget=R_Ear_FK_03
   properties {
    draw Style=Circle
    toolTipText=R_Ear_FK_02
    label Color {colorUseRule=UseColor_Specified, specifiedColor=<rgba 0 0 0 1>}
    normal Tint {colorUseRule=UseColor_Specified, specifiedColor=<rgba 1 0.0137 0 1>}
   }
   slot {
    class=CanvasPanelSlot
    properties {zOrder=1, layoutData {alignment=<0 0>, anchors {maximum=<0 0>, minimum=<0 0>}, offsets=<ltrb 636 16 54.882 30>}}
   }
  }
  R_Ear_FK_01 {
   widgetClassPath=&EUW_Picker_Button
   bIsVariable=true
   parent=CanvasPanel
   widget=R_Ear_FK_01
   properties {
    draw Style=Circle
    toolTipText=R_Ear_FK_01
    label Color {colorUseRule=UseColor_Specified, specifiedColor=<rgba 0 0 0 1>}
    normal Tint {colorUseRule=UseColor_Specified, specifiedColor=<rgba 1 0.0137 0 1>}
   }
   slot {
    class=CanvasPanelSlot
    properties {zOrder=1, layoutData {alignment=<0 0>, anchors {maximum=<0 0>, minimum=<0 0>}, offsets=<ltrb 636 88 65.178 30>}}
   }
  }
  R_Ear_Root_FK {
   widgetClassPath=&EUW_Picker_Button
   bIsVariable=true
   parent=CanvasPanel
   widget=R_Ear_Root_FK
   properties {
    draw Style=Circle
    toolTipText=R_Ear_FK_01
    label Color {colorUseRule=UseColor_Specified, specifiedColor=<rgba 0 0 0 1>}
    normal Tint {colorUseRule=UseColor_Specified, specifiedColor=<rgba 1 0.0137 0 1>}
   }
   slot {
    class=CanvasPanelSlot
    properties {zOrder=1, layoutData {alignment=<0 0>, anchors {maximum=<0 0>, minimum=<0 0>}, offsets=<ltrb 656 132 30 30>}}
   }
  }
  L_Ear_FK_02 {
   widgetClassPath=&EUW_Picker_Button
   bIsVariable=true
   parent=CanvasPanel
   widget=L_Ear_FK_02
   properties {
    draw Style=Circle
    toolTipText=L_Ear_FK_02
    label Color {colorUseRule=UseColor_Specified, specifiedColor=<rgba 0 0 0 1>}
    normal Tint {colorUseRule=UseColor_Specified, specifiedColor=<rgba 0 0.0735 1 1>}
   }
   slot {
    class=CanvasPanelSlot
    properties {zOrder=1, layoutData {alignment=<0 0>, anchors {maximum=<0 0>, minimum=<0 0>}, offsets=<ltrb 528 56 60.745 30>}}
   }
  }
  L_Ear_FK_03 {
   widgetClassPath=&EUW_Picker_Button
   bIsVariable=true
   parent=CanvasPanel
   widget=L_Ear_FK_03
   properties {
    draw Style=Circle
    toolTipText=L_Ear_FK_02
    label Color {colorUseRule=UseColor_Specified, specifiedColor=<rgba 0 0 0 1>}
    normal Tint {colorUseRule=UseColor_Specified, specifiedColor=<rgba 0 0.0735 1 1>}
   }
   slot {
    class=CanvasPanelSlot
    properties {zOrder=1, layoutData {alignment=<0 0>, anchors {maximum=<0 0>, minimum=<0 0>}, offsets=<ltrb 528 16 53.738 30>}}
   }
  }
  L_Ear_FK_01 {
   widgetClassPath=&EUW_Picker_Button
   bIsVariable=true
   parent=CanvasPanel
   widget=L_Ear_FK_01
   properties {
    draw Style=Circle
    toolTipText=L_Ear_FK_01
    label Color {colorUseRule=UseColor_Specified, specifiedColor=<rgba 0 0 0 1>}
    normal Tint {colorUseRule=UseColor_Specified, specifiedColor=<rgba 0 0.0735 1 1>}
   }
   slot {
    class=CanvasPanelSlot
    properties {zOrder=1, layoutData {alignment=<0 0>, anchors {maximum=<0 0>, minimum=<0 0>}, offsets=<ltrb 528 92 66.894 30.0001>}}
   }
  }
  L_Knee_PV {
   widgetClassPath=&EUW_Picker_Button
   bIsVariable=true
   parent=CanvasPanel
   widget=L_Knee_PV
   properties {
    draw Style=Circle
    toolTipText=L_Knee_PV
    label Color {colorUseRule=UseColor_Specified, specifiedColor=<rgba 0 0 0 1>}
    normal Tint {colorUseRule=UseColor_Specified, specifiedColor=<rgba 0 0.0735 1 1>}
   }
   slot {
    class=CanvasPanelSlot
    properties {zOrder=1, layoutData {alignment=<0 0>, anchors {maximum=<0 0>, minimum=<0 0>}, offsets=<ltrb 1147.976 799.2252 54.024 54.7748>}}
   }
  }
  R_Hip_Socket_Ctrl {
   widgetClassPath=&EUW_Picker_Button
   bIsVariable=true
   parent=CanvasPanel
   widget=R_Hip_Socket_Ctrl
   properties {
    draw Style=Circle
    toolTipText=R Hip Socket
    label Color {colorUseRule=UseColor_Specified, specifiedColor=<rgba 0 0 0 1>}
    normal Tint {colorUseRule=UseColor_Specified, specifiedColor=<rgba 1 0.9365 0 1>}
   }
   slot {
    class=CanvasPanelSlot
    properties {zOrder=1, layoutData {alignment=<0 0>, anchors {maximum=<0 0>, minimum=<0 0>}, offsets=<ltrb 1260 636 74 74>}}
   }
  }
  R_Scapula_Tip {
   widgetClassPath=&EUW_Picker_Button
   bIsVariable=true
   parent=CanvasPanel
   widget=R_Scapula_Tip
   properties {
    draw Style=Circle
    toolTipText=R Scapula Tip
    label Color {colorUseRule=UseColor_Specified, specifiedColor=<rgba 0 0 0 1>}
    normal Tint {colorUseRule=UseColor_Specified, specifiedColor=<rgba 1 0.9365 0 1>}
   }
   slot {
    class=CanvasPanelSlot
    properties {zOrder=1, layoutData {alignment=<0 0>, anchors {maximum=<0 0>, minimum=<0 0>}, offsets=<ltrb 784 640 74 74>}}
   }
  }
  L_Scapula_Tip {
   widgetClassPath=&EUW_Picker_Button
   bIsVariable=true
   parent=CanvasPanel
   widget=L_Scapula_Tip
   properties {
    draw Style=Circle
    toolTipText=L Scapula Tip
    label Color {colorUseRule=UseColor_Specified, specifiedColor=<rgba 0 0 0 1>}
    normal Tint {colorUseRule=UseColor_Specified, specifiedColor=<rgba 1 0.9365 0 1>}
   }
   slot {
    class=CanvasPanelSlot
    properties {zOrder=1, layoutData {alignment=<0 0>, anchors {maximum=<0 0>, minimum=<0 0>}, offsets=<ltrb 568 636 74 74>}}
   }
  }
  L_Hip_Socket_Ctrl {
   widgetClassPath=&EUW_Picker_Button
   bIsVariable=true
   parent=CanvasPanel
   widget=L_Hip_Socket_Ctrl
   properties {
    draw Style=Circle
    toolTipText=L Hip Socket
    label Color {colorUseRule=UseColor_Specified, specifiedColor=<rgba 0 0 0 1>}
    normal Tint {colorUseRule=UseColor_Specified, specifiedColor=<rgba 1 0.9365 0 1>}
   }
   slot {
    class=CanvasPanelSlot
    properties {zOrder=1, layoutData {alignment=<0 0>, anchors {maximum=<0 0>, minimum=<0 0>}, offsets=<ltrb 1080 640 74.4444 74.4444>}}
   }
  }
  R_Bk_Toe {
   widgetClassPath=&EUW_Picker_Button
   bIsVariable=true
   parent=CanvasPanel
   widget=R_Bk_Toe
   properties {
    draw Style=Square
    toolTipText=R Bk Toe Ctrl
    label Color {colorUseRule=UseColor_Specified, specifiedColor=<rgba 0 0 0 1>}
    normal Tint {colorUseRule=UseColor_Specified, specifiedColor=<rgba 1 0.4212 0 1>}
   }
   slot {
    class=CanvasPanelSlot
    properties {zOrder=1, layoutData {alignment=<0 0>, anchors {maximum=<0 0>, minimum=<0 0>}, offsets=<ltrb 1324 968.1659 33.1058 34.828>}}
   }
  }
  R_Bk_Ball_Pivot {
   widgetClassPath=&EUW_Picker_Button
   bIsVariable=true
   parent=CanvasPanel
   widget=R_Bk_Ball_Pivot
   properties {
    draw Style=Circle
    toolTipText=R Bk Ball Pivot Ctrl
    label Color {colorUseRule=UseColor_Specified, specifiedColor=<rgba 0 0 0 1>}
    normal Tint {colorUseRule=UseColor_Specified, specifiedColor=<rgba 0.0833 0.8226 1 1>}
   }
   slot {
    class=CanvasPanelSlot
    properties {zOrder=1, layoutData {alignment=<0 0>, anchors {maximum=<0 0>, minimum=<0 0>}, offsets=<ltrb 1292 972 26.012 26.558>}}
   }
  }
  L_Bk_Ik {
   widgetClassPath=&EUW_Picker_Button
   bIsVariable=true
   parent=CanvasPanel
   widget=L_Bk_Ik
   properties {
    draw Style=Square
    toolTipText=L Bk IK Ctrl
    label Color {colorUseRule=UseColor_Specified, specifiedColor=<rgba 0 0 0 1>}
    normal Tint {colorUseRule=UseColor_Specified, specifiedColor=<rgba 0 0 1 1>}
   }
   slot {
    class=CanvasPanelSlot
    properties {zOrder=1, layoutData {alignment=<0 0>, anchors {maximum=<0 0>, minimum=<0 0>}, offsets=<ltrb 1152 912 58 58>}}
   }
  }
  L_Bk_Ball_Pivot {
   widgetClassPath=&EUW_Picker_Button
   bIsVariable=true
   parent=CanvasPanel
   widget=L_Bk_Ball_Pivot
   properties {
    draw Style=Circle
    toolTipText=L Bk Ball Pivot Ctrl
    label Color {colorUseRule=UseColor_Specified, specifiedColor=<rgba 0 0 0 1>}
    normal Tint {colorUseRule=UseColor_Specified, specifiedColor=<rgba 0.0833 0.8226 1 1>}
   }
   slot {
    class=CanvasPanelSlot
    properties {zOrder=1, layoutData {alignment=<0 0>, anchors {maximum=<0 0>, minimum=<0 0>}, offsets=<ltrb 1080 936 26.5125 27.5135>}}
   }
  }
  L_Bk_Toe {
   widgetClassPath=&EUW_Picker_Button
   bIsVariable=true
   parent=CanvasPanel
   widget=L_Bk_Toe
   properties {
    draw Style=Square
    toolTipText=L Bk Toe Ctrl
    label Color {colorUseRule=UseColor_Specified, specifiedColor=<rgba 0 0 0 1>}
    normal Tint {colorUseRule=UseColor_Specified, specifiedColor=<rgba 1 0.4212 0 1>}
   }
   slot {
    class=CanvasPanelSlot
    properties {zOrder=1, layoutData {alignment=<0 0>, anchors {maximum=<0 0>, minimum=<0 0>}, offsets=<ltrb 1112 932 35.5175 35.5175>}}
   }
  }
  L_Bk_Ball {
   widgetClassPath=&EUW_Picker_Button
   bIsVariable=true
   parent=CanvasPanel
   widget=L_Bk_Ball
   properties {
    draw Style=Circle
    toolTipText=L Bk Ball Control
    label Color {colorUseRule=UseColor_Specified, specifiedColor=<rgba 0 0 0 1>}
    normal Tint {colorUseRule=UseColor_Specified, specifiedColor=<rgba 0 1 0.005 1>}
   }
   slot {
    class=CanvasPanelSlot
    properties {zOrder=1, layoutData {alignment=<0 0>, anchors {maximum=<0 0>, minimum=<0 0>}, offsets=<ltrb 1212 916 37.012 38.5135>}}
   }
  }
  TailGimbal {
   widgetClassPath=&EUW_Picker_Button
   bIsVariable=true
   parent=CanvasPanel
   widget=TailGimbal
   properties {
    draw Style=Circle
    toolTipText=Tail Gimbal Ctrl
    label Color {colorUseRule=UseColor_Specified, specifiedColor=<rgba 0 0 0 1>}
    normal Tint {colorUseRule=UseColor_Specified, specifiedColor=<rgba 0 0.2917 0.0173 1>}
    renderTransform {angle=0, scale=<1 1>, shear=<0 0>, translation=<-79.0791 54.054>}
   }
   slot {
    class=CanvasPanelSlot
    properties {zOrder=1, layoutData {alignment=<0 0>, anchors {maximum=<0 0>, minimum=<0 0>}, offsets=<ltrb 1352 341.998 33.989 31.987>}}
   }
  }
  tail_01 {
   widgetClassPath=&EUW_Picker_Button
   bIsVariable=true
   parent=CanvasPanel
   widget=tail_01
   properties {
    draw Style=Circle
    toolTipText=Tail 01
    label Color {colorUseRule=UseColor_Specified, specifiedColor=<rgba 0 0 0 1>}
    normal Tint {colorUseRule=UseColor_Specified, specifiedColor=<rgba 0 0.2917 0.0173 1>}
    renderTransform {angle=0, scale=<1 1>, shear=<0 0>, translation=<-79.0791 54.054>}
   }
   slot {
    class=CanvasPanelSlot
    properties {zOrder=1, layoutData {alignment=<0 0>, anchors {maximum=<0 0>, minimum=<0 0>}, offsets=<ltrb 1412 392 60 60>}}
   }
  }
  fk_04_Offset {
   widgetClassPath=&EUW_Picker_Button
   bIsVariable=true
   parent=CanvasPanel
   widget=fk_04_Offset
   properties {
    draw Style=RoundedSquareLargeRadius
    toolTipText=Tail FK 04 Offset
    label Color {colorUseRule=UseColor_Specified, specifiedColor=<rgba 0 0 0 1>}
    normal Tint {colorUseRule=UseColor_Specified, specifiedColor=<rgba 0 0.4043 1 1>}
    renderTransform {angle=0, scale=<1 1>, shear=<0 0>, translation=<-79.0791 54.054>}
   }
   slot {
    class=CanvasPanelSlot
    properties {zOrder=1, layoutData {alignment=<0 0>, anchors {maximum=<0 0>, minimum=<0 0>}, offsets=<ltrb 1644 344 30 30>}}
   }
  }
  fk_03_Offset {
   widgetClassPath=&EUW_Picker_Button
   bIsVariable=true
   parent=CanvasPanel
   widget=fk_03_Offset
   properties {
    draw Style=RoundedSquareLargeRadius
    toolTipText=Tail FK 03 Offset
    label Color {colorUseRule=UseColor_Specified, specifiedColor=<rgba 0 0 0 1>}
    normal Tint {colorUseRule=UseColor_Specified, specifiedColor=<rgba 0 0.4043 1 1>}
    renderTransform {angle=0, scale=<1 1>, shear=<0 0>, translation=<-79.0791 54.054>}
   }
   slot {
    class=CanvasPanelSlot
    properties {zOrder=1, layoutData {alignment=<0 0>, anchors {maximum=<0 0>, minimum=<0 0>}, offsets=<ltrb 1572 344 30 30>}}
   }
  }
  fk_02_Offset {
   widgetClassPath=&EUW_Picker_Button
   bIsVariable=true
   parent=CanvasPanel
   widget=fk_02_Offset
   properties {
    draw Style=RoundedSquareLargeRadius
    toolTipText=Tail FK 02 Offset
    label Color {colorUseRule=UseColor_Specified, specifiedColor=<rgba 0 0 0 1>}
    normal Tint {colorUseRule=UseColor_Specified, specifiedColor=<rgba 0 0.4043 1 1>}
    renderTransform {angle=0, scale=<1 1>, shear=<0 0>, translation=<-79.0791 54.054>}
   }
   slot {
    class=CanvasPanelSlot
    properties {zOrder=1, layoutData {alignment=<0 0>, anchors {maximum=<0 0>, minimum=<0 0>}, offsets=<ltrb 1500 344 30 30>}}
   }
  }
  fk_01_Offset {
   widgetClassPath=&EUW_Picker_Button
   bIsVariable=true
   parent=CanvasPanel
   widget=fk_01_Offset
   properties {
    draw Style=RoundedSquareLargeRadius
    toolTipText=Tail FK 01 Offset
    label Color {colorUseRule=UseColor_Specified, specifiedColor=<rgba 0 0 0 1>}
    normal Tint {colorUseRule=UseColor_Specified, specifiedColor=<rgba 0 0.4043 1 1>}
    renderTransform {angle=0, scale=<1 1>, shear=<0 0>, translation=<-79.0791 54.054>}
   }
   slot {
    class=CanvasPanelSlot
    properties {zOrder=1, layoutData {alignment=<0 0>, anchors {maximum=<0 0>, minimum=<0 0>}, offsets=<ltrb 1428 344 30 30>}}
   }
  }
  tail_02 {
   widgetClassPath=&EUW_Picker_Button
   bIsVariable=true
   parent=CanvasPanel
   widget=tail_02
   properties {
    draw Style=Circle
    toolTipText=Tail 02 Ctrl
    label Color {colorUseRule=UseColor_Specified, specifiedColor=<rgba 0 0 0 1>}
    normal Tint {colorUseRule=UseColor_Specified, specifiedColor=<rgba 0.101 1 0 1>}
    renderTransform {angle=0, scale=<1 1>, shear=<0 0>, translation=<-79.0791 54.054>}
   }
   slot {
    class=CanvasPanelSlot
    properties {zOrder=1, layoutData {alignment=<0 0>, anchors {maximum=<0 0>, minimum=<0 0>}, offsets=<ltrb 1484 392 60 60>}}
   }
  }
  tail_03 {
   widgetClassPath=&EUW_Picker_Button
   bIsVariable=true
   parent=CanvasPanel
   widget=tail_03
   properties {
    draw Style=Circle
    toolTipText=Tail 03 Ctrl
    label Color {colorUseRule=UseColor_Specified, specifiedColor=<rgba 0 0 0 1>}
    normal Tint {colorUseRule=UseColor_Specified, specifiedColor=<rgba 0 0.2917 0.0173 1>}
    renderTransform {angle=0, scale=<1 1>, shear=<0 0>, translation=<-79.0791 54.054>}
   }
   slot {
    class=CanvasPanelSlot
    properties {zOrder=1, layoutData {alignment=<0 0>, anchors {maximum=<0 0>, minimum=<0 0>}, offsets=<ltrb 1556 392 60 60>}}
   }
  }
  tail_04 {
   widgetClassPath=&EUW_Picker_Button
   bIsVariable=true
   parent=CanvasPanel
   widget=tail_04
   properties {
    draw Style=Circle
    toolTipText=Tail 04 Ctrl
    label Color {colorUseRule=UseColor_Specified, specifiedColor=<rgba 0 0 0 1>}
    normal Tint {colorUseRule=UseColor_Specified, specifiedColor=<rgba 0.101 1 0 1>}
    renderTransform {angle=0, scale=<1 1>, shear=<0 0>, translation=<-79.0791 54.054>}
   }
   slot {
    class=CanvasPanelSlot
    properties {zOrder=1, layoutData {alignment=<0 0>, anchors {maximum=<0 0>, minimum=<0 0>}, offsets=<ltrb 1628 392 60 60>}}
   }
  }
  TextBlock_52 {
   widgetClassPath=&TextBlock
   parent=CanvasPanel
   widget=TextBlock_52
   properties {text=L}
   slot {class=CanvasPanelSlot, properties {layoutData {alignment=<0 0>, anchors {maximum=<0 0>, minimum=<0 0>}, offsets=<ltrb 448 952 30 30>}}}
  }
  TextBlock_2 {
   widgetClassPath=&TextBlock
   parent=CanvasPanel
   widget=TextBlock_2
   properties {text=L}
   slot {class=CanvasPanelSlot, properties {layoutData {alignment=<0 0>, anchors {maximum=<0 0>, minimum=<0 0>}, offsets=<ltrb 1052 944 30 30>}}}
  }
  TextBlock {
   widgetClassPath=&TextBlock
   parent=CanvasPanel
   widget=TextBlock
   properties {text=R}
   slot {class=CanvasPanelSlot, properties {layoutData {alignment=<0 0>, anchors {maximum=<0 0>, minimum=<0 0>}, offsets=<ltrb 896 984 48 48.7688>}}}
  }
  TextBlock_3 {
   widgetClassPath=&TextBlock
   parent=CanvasPanel
   widget=TextBlock_3
   properties {text=R}
   slot {class=CanvasPanelSlot, properties {layoutData {alignment=<0 0>, anchors {maximum=<0 0>, minimum=<0 0>}, offsets=<ltrb 1472 980 48 48>}}}
  }
  Image_1 {
   widgetClassPath=&Image
   bIsVariable=true
   parent=CanvasPanel
   widget=Image_1
   properties {
    brush {
     bIsDynamicallyLoaded=false
     drawAs=Image
     imageType=FullColor
     mirroring=NoMirror
     resourceName=None
     resourceObject=null
     tiling=NoTile
     imageSize=<1547 1058>
     margin=<ltrb 0 0 0 0>
     outlineSettings {
      bUseBrushTransparency=false
      roundingType=HalfHeightRadius
      width=0
      color {colorUseRule=UseColor_Specified, specifiedColor=<rgba 0 0 0 0>}
      cornerRadii=<0 0 0 0>
     }
     tintColor {colorUseRule=UseColor_Specified, specifiedColor=<rgba 1 1 1 1>}
     uVRegion {bIsValid=false, max=<0 0>, min=<0 0>}
    }
    colorAndOpacity=<rgba 1 1 1 0.336>
   }
   slot {
    class=CanvasPanelSlot
    properties {layoutData {alignment=<0 0>, anchors {maximum=<0 0>, minimum=<0 0>}, offsets=<ltrb -64 -40 2031.4872 1185.3676>}}
   }
  }
 ]
}
