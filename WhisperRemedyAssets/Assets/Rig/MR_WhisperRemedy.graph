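# Modular Control Rig for WhisperRemedy, a caped quadruped character with tail, ears and collar tag.
# The event graph holds only an empty begin-execution node, while namespaced connectors mark modules for root, spine, neck, four legs, tail, face, ears and FK chains including the cape.
# The skeleton includes front and rear IK foot bones and spine-parented target bones for cape segments.
# Pose curves cover facial expression, eyes, jaw and tongue.
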ControlRigBlueprint /WhisperRemedyAssets/Assets/Rig/MR_WhisperRemedy
graphs [
 RigVMFunctionLibrary {links [], nodes []}
 RigVMModel {nodes [RigUnit_BeginExecution = &RigUnit_BeginExecution()]}
]
hierarchy [
 root {loc=<0 0 0>, rot=<0 0 0 1>}
 pelvis {parent=root, loc=<0 -15.3471 38.2106>, rot=<0.4921 0.5078 -0.4921 -0.5078>}
 spine_01 {parent=pelvis, loc=<0 0 0>, rot=<0 0 -0.0124 -0.9999>}
 spine_02 {parent=spine_01, loc=<13.8833 -0.3457 0>, rot=<0 0 0.0137 -0.9999>}
 spine_03 {parent=spine_02, loc=<15.5639 0 0>, rot=<0 0 -0.011 -0.9999>}
 shoulder_fr_l {parent=spine_03, loc=<3.186 4.1292 -7.8083>, rot=<0.0037 0.0034 -0.8214 -0.5703>}
 elbow_fr_l {parent=shoulder_fr_l, loc=<13.7779 0 0>, rot=<-0.0001 -0.0001 0.1796 -0.9837>}
 wrist_fr_l {parent=elbow_fr_l, loc=<15.1156 0 0>, rot=<0.994 -0.0933 0.0203 -0.0527>}
 ball_fr_l {parent=wrist_fr_l, loc=<4.5512 0 0>, rot=<-0.0153 0.0292 -0.3509 -0.9358>}
 toe_fr_l {parent=ball_fr_l, loc=<1.9055 0 0>, rot=<-0.0007 -0.0415 -0.2548 -0.9661>}
 toe_tip_fr_l {parent=toe_fr_l, loc=<4.6825 0 0>, rot=<0.0046 0.0116 -0.0519 -0.9986>}
 scapula_fr_l {parent=shoulder_fr_l, loc=<-6.9459 3.8453 -0.005>, rot=<-0.0048 0.0016 0.2501 -0.9682>}
 shoulder_fr_r {parent=spine_03, loc=<3.186 4.1293 7.8084>, rot=<0.0034 -0.0037 0.5703 -0.8214>}
 elbow_fr_r {parent=shoulder_fr_r, loc=<-13.7779 0 -0.0001>, rot=<-0.0001 -0.0001 0.1796 -0.9837>}
 wrist_fr_r {parent=elbow_fr_r, loc=<-14.4175 -0.1127 0.0153>, rot=<0.994 -0.0933 0.0203 -0.0527>}
 ball_fr_r {parent=wrist_fr_r, loc=<-5.2581 0.0196 -0.0085>, rot=<-0.0153 0.0292 -0.3509 -0.9358>}
 toe_fr_r {parent=ball_fr_r, loc=<-1.9055 0 0>, rot=<-0.0007 -0.0415 -0.2548 -0.9661>}
 toe_tip_fr_r {parent=toe_fr_r, loc=<-4.6824 0 0.0001>, rot=<0.0046 0.0116 -0.0519 -0.9986>}
 scapula_fr_r {parent=shoulder_fr_r, loc=<6.9459 -3.8454 0.005>, rot=<-0.0048 0.0016 0.2501 -0.9682>}
 neck_01 {parent=spine_03, loc=<1.1967 -0.0091 0>, rot=<0 0 0.4992 -0.8665>}
 neck_02 {parent=neck_01, loc=<13.2288 0 0>, rot=<0 0 0.0244 -0.9997>}
 head {parent=neck_02, loc=<10.6145 0 0>, rot=<-0.1543 -0.6901 0.1543 -0.6901>}
 face_attach {parent=head, loc=<0 0 0>, rot=<0 0 0 1>}
 jaw {parent=face_attach, loc=<0 1.5414 1.5673>, rot=<0 0.7071 0 -0.7071>}
 lip_lower_03_l {parent=jaw, loc=<-7.7319 9.0929 -3.1083>, rot=<-0.0567 -0.789 -0.0901 -0.605>}
 lip_lower_02_l {parent=jaw, loc=<-7.7793 10.6851 -2.7467>, rot=<-0.0802 -0.728 -0.1069 -0.6725>}
 lip_lower_01_l {parent=jaw, loc=<-7.8737 12.1434 -1.5952>, rot=<-0.1854 -0.7063 -0.1903 -0.6561>}
 lip_lower_mid {parent=jaw, loc=<-7.9235 13.3354 0>, rot=<0.4939 0.506 0.506 0.4939>}
 lip_lower_03_r {parent=jaw, loc=<-7.7318 9.0929 3.1083>, rot=<0.789 -0.0567 -0.605 0.0901>}
 lip_lower_01_r {parent=jaw, loc=<-7.8738 12.1435 1.5952>, rot=<0.7063 -0.1854 -0.6561 0.1903>}
 lip_lower_02_r {parent=jaw, loc=<-7.7793 10.6851 2.7467>, rot=<0.728 -0.0802 -0.6725 0.1069>}
 teeth_lower {parent=jaw, loc=<-7.9764 4.8748 0>, rot=<0 -0.7071 0 -0.7071>}
 tongue_01 {parent=jaw, loc=<-5.8122 -0.4681 0>, rot=<0.5405 -0.4559 0.4559 -0.5405>}
 tongue_02 {parent=tongue_01, loc=<-5.4528 0 0.1308>, rot=<0 -0.04 0 -0.9992>}
 tongue_03 {parent=tongue_02, loc=<-3.7239 0 0.1308>, rot=<0 -0.0699 0 -0.9976>}
 lip_corner_l {parent=face_attach, loc=<3.532 9.0864 -5.4803>, rot=<0.0323 -0.1402 -0.1027 -0.9843>}
 cheek_bt_l {parent=face_attach, loc=<5.2759 4.981 -3.3543>, rot=<0.0975 -0.0174 -0.1751 -0.9796>}
 lip_corner_r {parent=face_attach, loc=<-3.532 9.0864 -5.4803>, rot=<0.9843 -0.1027 0.1402 0.0323>}
 cheek_bt_r {parent=face_attach, loc=<-5.2759 4.981 -3.3542>, rot=<0.9796 -0.1751 0.0174 0.0975>}
 reverse_jaw {parent=face_attach, loc=<0 0.5985 -5.1509>, rot=<0 0.7071 0 -0.7071>}
 brow_ot_l {parent=reverse_jaw, loc=<8.808 4.5312 -5.3802>, rot=<0.1966 -0.6792 0.1966 -0.6792>}
 brow_ot_r {parent=reverse_jaw, loc=<8.808 4.5312 5.3802>, rot=<-0.6792 -0.1966 0.6792 0.1966>}
 brow_in_r {parent=reverse_jaw, loc=<9.7245 6.9681 1.8436>, rot=<-0.6991 -0.1061 0.6991 0.1061>}
 brow_in_l {parent=reverse_jaw, loc=<9.7244 6.9681 -1.8436>, rot=<0.1061 -0.6991 0.1061 -0.6991>}
 brow_ct {parent=reverse_jaw, loc=<8.3974 8.8869 0>, rot=<0.1061 -0.6991 0.1061 -0.6991>}
 brow_md_l {parent=reverse_jaw, loc=<9.9048 6.1057 -4.3125>, rot=<0.1929 -0.6803 0.1929 -0.6803>}
 brow_md_r {parent=reverse_jaw, loc=<9.9048 6.1058 4.3126>, rot=<-0.6803 -0.1929 0.6803 0.1929>}
 lip_upper_03_l {parent=reverse_jaw, loc=<0.129 9.9627 -3.7462>, rot=<-0.0562 -0.7924 -0.0904 -0.6006>}
 lip_upper_02_l {parent=reverse_jaw, loc=<0.3249 12.1056 -3.5072>, rot=<-0.0539 -0.7311 -0.0889 -0.6742>}
 lip_upper_01_l {parent=reverse_jaw, loc=<0.5666 14.2448 -2.5853>, rot=<-0.195 -0.707 -0.219 -0.6436>}
 cheek_tp_ot_l {parent=reverse_jaw, loc=<4.5625 5.3738 -4.7949>, rot=<-0.1358 -0.6932 -0.1692 -0.6874>}
 cheek_tp_in_l {parent=reverse_jaw, loc=<5.1286 8.5302 -1.8132>, rot=<-0.2594 -0.6712 -0.2639 -0.6423>}
 lip_upper_01_r {parent=reverse_jaw, loc=<0.5665 14.2448 2.5854>, rot=<0.707 -0.195 -0.6436 0.219>}
 lip_upper_02_r {parent=reverse_jaw, loc=<0.3249 12.1056 3.5072>, rot=<0.7311 -0.0539 -0.6742 0.0889>}
 lip_upper_03_r {parent=reverse_jaw, loc=<0.129 9.9626 3.7462>, rot=<0.7924 -0.0562 -0.6006 0.0904>}
 cheek_tp_ot_r {parent=reverse_jaw, loc=<4.5624 5.3738 4.7949>, rot=<0.6932 -0.1358 -0.6874 0.1692>}
 cheek_tp_in_r {parent=reverse_jaw, loc=<5.1285 8.5302 1.8132>, rot=<0.6712 -0.2594 -0.6423 0.2639>}
 ear_01_l {parent=reverse_jaw, loc=<7.5029 -0.0902 -0.7241>, rot=<0.1625 -0.4789 0.1881 -0.842>}
 ear_02_l {parent=ear_01_l, loc=<6.6137 0 0>, rot=<-0.1396 -0.1853 -0.149 -0.9612>}
 ear_03_l {parent=ear_02_l, loc=<3.9918 0 0>, rot=<-0.1664 -0.2168 0.0087 -0.9619>}
 ear_04_l {parent=ear_03_l, loc=<3.9148 0 0>, rot=<0.0672 -0.1399 0.0639 -0.9858>}
 ear_01_r {parent=reverse_jaw, loc=<7.5029 -0.0902 0.7241>, rot=<-0.4789 -0.1625 0.842 0.1881>}
 ear_02_r {parent=ear_01_r, loc=<-6.2909 -0.561 0.6142>, rot=<-0.1707 -0.6662 -0.3078 -0.6575>}
 ear_03_r {parent=ear_02_r, loc=<-4.7445 -0.0001 0>, rot=<0.0117 -0.2533 0.0982 -0.9623>}
 ear_04_r {parent=ear_03_r, loc=<-5.0608 0.0001 -0.0001>, rot=<-0.0008 -0.0406 -0.0207 -0.999>}
 lip_upper_mid {parent=reverse_jaw, loc=<0.4164 15.4191 0>, rot=<-0.4824 -0.517 -0.517 -0.4824>}
 eye_r {parent=reverse_jaw, loc=<7.2257 6.2822 3.0323>, rot=<0.693 0.1403 -0.6874 -0.1657>}
 eye_l {parent=reverse_jaw, loc=<7.2257 6.2822 -3.0323>, rot=<0.1403 -0.693 0.1657 -0.6874>}
 lid_bt_r {parent=reverse_jaw, loc=<7.2257 6.2822 3.0323>, rot=<-0.6903 -0.153 0.6903 0.153>}
 lid_tp_r {parent=reverse_jaw, loc=<7.2257 6.2822 3.0323>, rot=<-0.6903 -0.153 0.6903 0.153>}
 nose_mid {parent=reverse_jaw, loc=<3.1478 16.8243 0>, rot=<-0.546 -0.4493 -0.4493 -0.546>}
 nostril_l {parent=nose_mid, loc=<0.2659 -1.23 0.259>, rot=<0 -0.2275 0 -0.9738>}
 nostril_r {parent=nose_mid, loc=<0.2659 1.23 0.259>, rot=<0 -0.9738 0 0.2275>}
 lid_tp_l {parent=reverse_jaw, loc=<7.2257 6.2822 -3.0323>, rot=<0.153 -0.6903 0.153 -0.6903>}
 lid_bt_l {parent=reverse_jaw, loc=<7.2257 6.2822 -3.0323>, rot=<0.153 -0.6903 0.153 -0.6903>}
 teeth_upper {parent=reverse_jaw, loc=<2.2634 6.6577 0>, rot=<0 -0.7071 0 -0.7071>}
 dyn_collar_tag_MSTR {parent=neck_02, loc=<-1.0382 7.9049 0.0282>, rot=<0 0 0.9731 0.2303>}
 dyn_collar_tag {parent=dyn_collar_tag_MSTR, loc=<0.3 0 0>, rot=<0 0 0 1>}
 dyn_cape_l_01_MSTR {parent=neck_01, loc=<10.2977 -4.8526 -6.5271>, rot=<0.1554 -0.9597 0.1773 -0.1529>}
 dyn_cape_l_01 {parent=dyn_cape_l_01_MSTR, loc=<0.2195 0 0>, rot=<0 0 0 1>}
 dyn_cape_l_02 {parent=dyn_cape_l_01, loc=<8.2064 0 0>, rot=<-0.0017 0.0195 0.1038 -0.9944>}
 dyn_cape_l_03 {parent=dyn_cape_l_02, loc=<9.1515 0 0>, rot=<0 0.0053 0.079 -0.9969>}
 dyn_cape_l_04 {parent=dyn_cape_l_03, loc=<11.112 0 0>, rot=<0.0022 -0.0239 0.0928 -0.9954>}
 dyn_cape_mid_01_MSTR {parent=neck_01, loc=<11.6817 -8.6056 0.0282>, rot=<0.2439 -0.6637 0.6637 -0.2439>}
 dyn_cape_mid_01 {parent=dyn_cape_mid_01_MSTR, loc=<0.2195 0 0>, rot=<0 0 0 1>}
 dyn_cape_mid_02 {parent=dyn_cape_mid_01, loc=<6.8432 0 0>, rot=<0 0.0686 0 -0.9976>}
 dyn_cape_mid_03 {parent=dyn_cape_mid_02, loc=<8.9792 0 0>, rot=<0 0.05 0 -0.9987>}
 dyn_cape_mid_04 {parent=dyn_cape_mid_03, loc=<10.4235 0 0>, rot=<0 0.0033 0 -1>}
 dyn_cape_r_01_MSTR {parent=neck_01, loc=<10.2977 -4.8526 6.5271>, rot=<0.1529 -0.1773 0.9597 -0.1554>}
 dyn_cape_r_01 {parent=dyn_cape_r_01_MSTR, loc=<0.2194 -0.0001 0>, rot=<0 0 0 1>}
 dyn_cape_r_02 {parent=dyn_cape_r_01, loc=<8.2064 0.0001 0>, rot=<0.0017 0.0195 -0.1038 -0.9944>}
 dyn_cape_r_03 {parent=dyn_cape_r_02, loc=<9.1515 0 0>, rot=<0 0.0053 -0.079 -0.9969>}
 dyn_cape_r_04 {parent=dyn_cape_r_03, loc=<11.112 -0.0001 0>, rot=<-0.0022 -0.0239 -0.0928 -0.9954>}
 dyn_cape_l_02_tgt_MSTR {parent=spine_03, loc=<-3.966 -6.3776 -8.479>, rot=<-0.2437 -0.9386 0.2415 -0.037>}
 dyn_cape_l_02_tgt {parent=dyn_cape_l_02_tgt_MSTR, loc=<0.3 0 0>, rot=<0 0 0 1>}
 dyn_cape_mid_02_tgt_MSTR {parent=spine_03, loc=<-6.7493 -12.1308 0.0282>, rot=<-0.0718 -0.7034 0.7034 0.0718>}
 dyn_cape_mid_02_tgt {parent=dyn_cape_mid_02_tgt_MSTR, loc=<0.3 0 0>, rot=<0 0 0 1>}
 dyn_cape_r_02_tgt_MSTR {parent=spine_03, loc=<-3.966 -6.3776 8.479>, rot=<0.037 -0.2415 0.9386 0.2437>}
 dyn_cape_r_02_tgt {parent=dyn_cape_r_02_tgt_MSTR, loc=<0.3 0 0>, rot=<0 0 0 1>}
 dyn_cape_l_03_tgt_MSTR {parent=spine_02, loc=<3.6599 -2.0969 -9.3189>, rot=<-0.1569 -0.9565 0.2447 -0.0255>}
 dyn_cape_l_03_tgt {parent=dyn_cape_l_03_tgt_MSTR, loc=<0.3 0 1.0137>, rot=<0 0 0 1>}
 dyn_cape_mid_03_tgt_MSTR {parent=spine_02, loc=<0.2559 -10.6261 0.0282>, rot=<-0.0288 -0.7065 0.7065 0.0288>}
 dyn_cape_mid_03_tgt {parent=dyn_cape_mid_03_tgt_MSTR, loc=<0.3 0 0>, rot=<0 0 0 1>}
 dyn_cape_r_03_tgt_MSTR {parent=spine_02, loc=<3.6599 -2.0969 9.3189>, rot=<0.0255 -0.2447 0.9565 0.1569>}
 dyn_cape_r_03_tgt {parent=dyn_cape_r_03_tgt_MSTR, loc=<0.3 0 1.0137>, rot=<0 0 0 1>}
 dyn_cape_l_04_tgt_MSTR {parent=spine_01, loc=<7.3031 2.1336 -8.5053>, rot=<-0.0865 -0.9667 0.2398 0.0232>}
 dyn_cape_l_04_tgt {parent=dyn_cape_l_04_tgt_MSTR, loc=<0.3 0 3.5499>, rot=<0 0 0 1>}
 dyn_cape_mid_04_tgt_MSTR {parent=spine_01, loc=<3.4855 -9.8394 0.0282>, rot=<-0.0362 -0.7062 0.7062 0.0362>}
 dyn_cape_mid_04_tgt {parent=dyn_cape_mid_04_tgt_MSTR, loc=<0.3 0 0>, rot=<0 0 0 1>}
 dyn_cape_r_04_tgt_MSTR {parent=spine_01, loc=<7.3031 2.1336 8.5054>, rot=<-0.0232 -0.2398 0.9667 0.0865>}
 dyn_cape_r_04_tgt {parent=dyn_cape_r_04_tgt_MSTR, loc=<0.3 0 3.5499>, rot=<0 0 0 1>}
 hips {parent=pelvis, loc=<0 0 0>, rot=<0 0 0 1>}
 tail_01 {parent=hips, loc=<-7.6516 -5.3072 0>, rot=<0.0124 -0.707 0.7071 -0.0069>}
 tail_02 {parent=tail_01, loc=<8.4235 0 0>, rot=<0.0019 -0.049 -0.0364 -0.9981>}
 tail_03 {parent=tail_02, loc=<8.4131 0 0>, rot=<-0.0024 0.0623 0.061 -0.9962>}
 tail_04 {parent=tail_03, loc=<8.4174 0 0>, rot=<0.0004 0.0102 0.0126 -0.9999>}
 thigh_bk_l {parent=hips, loc=<-4.5411 0.7536 -6.4095>, rot=<-0.6756 -0.7373 0 0>}
 knee_bk_l {parent=thigh_bk_l, loc=<15.1791 0 0>, rot=<0 0 0.3983 -0.9172>}
 ankle_bk_l {parent=knee_bk_l, loc=<12.2939 0 0>, rot=<0 0 -0.4247 -0.9053>}
 ball_bk_l {parent=ankle_bk_l, loc=<11.5692 0 0>, rot=<0 0 -0.4268 -0.9043>}
 toe_bk_l {parent=ball_bk_l, loc=<1.4603 0 0>, rot=<0 0 -0.1105 -0.9939>}
 toe_tip_bk_l {parent=toe_bk_l, loc=<3.6201 0 0>, rot=<0 0 -0.2748 -0.9615>}
 thigh_bk_r {parent=hips, loc=<-4.5411 0.7536 6.4095>, rot=<0.7373 -0.6756 0 0>}
 knee_bk_r {parent=thigh_bk_r, loc=<-15.1792 0 0>, rot=<0 0 0.3983 -0.9172>}
 ankle_bk_r {parent=knee_bk_r, loc=<-12.2939 0 0>, rot=<0 0 -0.4247 -0.9053>}
 ball_bk_r {parent=ankle_bk_r, loc=<-11.5691 0 0>, rot=<0 0 -0.4268 -0.9043>}
 toe_bk_r {parent=ball_bk_r, loc=<-1.4603 0 0>, rot=<0 0 -0.1105 -0.9939>}
 toe_tip_bk_r {parent=toe_bk_r, loc=<-3.6201 0 0>, rot=<0 0 -0.2748 -0.9615>}
 IK_Foot_Rear_Root {parent=root, loc=<0 -30.9193 0>, rot=<0 0 0 1>}
 IK_Foot_Rear_L {parent=IK_Foot_Rear_Root, loc=<6.4093 0.7896 2.7138>, rot=<-0.3024 -0.6392 -0.3024 -0.6392>}
 IK_Foot_Rear_R {parent=IK_Foot_Rear_Root, loc=<-6.4093 0.7896 2.7138>, rot=<0.6392 -0.3024 0.6392 -0.3024>}
 IK_Foot_Front_Root {parent=root, loc=<0 4.0614 0>, rot=<0.135 0 0 -0.9908>}
 IK_Foot_Front_L {parent=IK_Foot_Front_Root, loc=<7.6694 8.7288 5.0698>, rot=<-0.4464 -0.5742 -0.3431 -0.5944>}
 IK_Foot_Front_R {parent=IK_Foot_Front_Root, loc=<-7.6694 8.7288 5.0698>, rot=<0.5944 -0.3431 0.5742 -0.4464>}
 BothLegIK {type=Curve, loc=<0 0 0>, rot=<0 0 0 1>}
 jaw_open_pose {type=Curve, loc=<0 0 0>, rot=<0 0 0 1>}
 reverse_jaw_pose {type=Curve, loc=<0 0 0>, rot=<0 0 0 1>}
 L_lip_corner_narrow {type=Curve, loc=<0 0 0>, rot=<0 0 0 1>}
 R_lip_corner_narrow {type=Curve, loc=<0 0 0>, rot=<0 0 0 1>}
 L_lip_corner_wide {type=Curve, loc=<0 0 0>, rot=<0 0 0 1>}
 R_lip_corner_wide {type=Curve, loc=<0 0 0>, rot=<0 0 0 1>}
 L_smile {type=Curve, loc=<0 0 0>, rot=<0 0 0 1>}
 R_smile {type=Curve, loc=<0 0 0>, rot=<0 0 0 1>}
 L_frown {type=Curve, loc=<0 0 0>, rot=<0 0 0 1>}
 R_frown {type=Curve, loc=<0 0 0>, rot=<0 0 0 1>}
 L_lower_lip_down {type=Curve, loc=<0 0 0>, rot=<0 0 0 1>}
 R_lower_lip_down {type=Curve, loc=<0 0 0>, rot=<0 0 0 1>}
 L_lower_lip_up {type=Curve, loc=<0 0 0>, rot=<0 0 0 1>}
 R_lower_lip_up {type=Curve, loc=<0 0 0>, rot=<0 0 0 1>}
 R_upper_lip_down {type=Curve, loc=<0 0 0>, rot=<0 0 0 1>}
 L_upper_lip_down {type=Curve, loc=<0 0 0>, rot=<0 0 0 1>}
 L_upper_lip_up {type=Curve, loc=<0 0 0>, rot=<0 0 0 1>}
 R_upper_lip_up {type=Curve, loc=<0 0 0>, rot=<0 0 0 1>}
 tongue_back {type=Curve, loc=<0 0 0>, rot=<0 0 0 1>}
 tongue_forward {type=Curve, loc=<0 0 0>, rot=<0 0 0 1>}
 tongue_up {type=Curve, loc=<0 0 0>, rot=<0 0 0 1>}
 tongue_down {type=Curve, loc=<0 0 0>, rot=<0 0 0 1>}
 Reaction_Happy {type=Curve, loc=<0 0 0>, rot=<0 0 0 1>}
 look_right_pose {type=Curve, loc=<0 0 0>, rot=<0 0 0 1>}
 L_wide_pose {type=Curve, loc=<0 0 0>, rot=<0 0 0 1>}
 R_lip_corner_narrow_pose {type=Curve, loc=<0 0 0>, rot=<0 0 0 1>}
 R_upper_lip_lower_pose {type=Curve, loc=<0 0 0>, rot=<0 0 0 1>}
 L_lip_corner_narrow_pose {type=Curve, loc=<0 0 0>, rot=<0 0 0 1>}
 L_upper_lip_lower_pose {type=Curve, loc=<0 0 0>, rot=<0 0 0 1>}
 R_lower_lip_up_pose {type=Curve, loc=<0 0 0>, rot=<0 0 0 1>}
 R_brow_squeeze_in_pose {type=Curve, loc=<0 0 0>, rot=<0 0 0 1>}
 L_lip_corner_wide_pose {type=Curve, loc=<0 0 0>, rot=<0 0 0 1>}
 L_upperLidDown_pose {type=Curve, loc=<0 0 0>, rot=<0 0 0 1>}
 R_lower_lip_down_pose {type=Curve, loc=<0 0 0>, rot=<0 0 0 1>}
 R_upper_lip_raiser_pose {type=Curve, loc=<0 0 0>, rot=<0 0 0 1>}
 L_lower_lip_down_pose {type=Curve, loc=<0 0 0>, rot=<0 0 0 1>}
 L_upper_lip_raiser_pose {type=Curve, loc=<0 0 0>, rot=<0 0 0 1>}
 L_brow_squeeze_in_pose {type=Curve, loc=<0 0 0>, rot=<0 0 0 1>}
 L_lower_lip_up_pose {type=Curve, loc=<0 0 0>, rot=<0 0 0 1>}
 R_upperLidDown_pose {type=Curve, loc=<0 0 0>, rot=<0 0 0 1>}
 R_wide_pose {type=Curve, loc=<0 0 0>, rot=<0 0 0 1>}
 R_blink_pose {type=Curve, loc=<0 0 0>, rot=<0 0 0 1>}
 look_down_pose {type=Curve, loc=<0 0 0>, rot=<0 0 0 1>}
 look_up_pose {type=Curve, loc=<0 0 0>, rot=<0 0 0 1>}
 L_blink_pose {type=Curve, loc=<0 0 0>, rot=<0 0 0 1>}
 look_left_pose {type=Curve, loc=<0 0 0>, rot=<0 0 0 1>}
 R_lip_corner_wide_pose {type=Curve, loc=<0 0 0>, rot=<0 0 0 1>}
 L_smile_pose {type=Curve, loc=<0 0 0>, rot=<0 0 0 1>}
 R_smile_pose {type=Curve, loc=<0 0 0>, rot=<0 0 0 1>}
 L_frown_pose {type=Curve, loc=<0 0 0>, rot=<0 0 0 1>}
 R_frown_pose {type=Curve, loc=<0 0 0>, rot=<0 0 0 1>}
 C_nose_wrinkler_pose {type=Curve, loc=<0 0 0>, rot=<0 0 0 1>}
 L_cheek_up_pose {type=Curve, loc=<0 0 0>, rot=<0 0 0 1>}
 R_cheek_up_pose {type=Curve, loc=<0 0 0>, rot=<0 0 0 1>}
 jaw_left_pose {type=Curve, loc=<0 0 0>, rot=<0 0 0 1>}
 jaw_right_pose {type=Curve, loc=<0 0 0>, rot=<0 0 0 1>}
 jaw_up_pose {type=Curve, loc=<0 0 0>, rot=<0 0 0 1>}
 jaw_forward_pose {type=Curve, loc=<0 0 0>, rot=<0 0 0 1>}
 jaw_back_pose {type=Curve, loc=<0 0 0>, rot=<0 0 0 1>}
 eye_convergence_pose {type=Curve, loc=<0 0 0>, rot=<0 0 0 1>}
 tongue_forward_pose {type=Curve, loc=<0 0 0>, rot=<0 0 0 1>}
 tongue_back_pose {type=Curve, loc=<0 0 0>, rot=<0 0 0 1>}
 tongue_up_pose {type=Curve, loc=<0 0 0>, rot=<0 0 0 1>}
 tongue_down_pose {type=Curve, loc=<0 0 0>, rot=<0 0 0 1>}
 L_brow_In_Up_pose {type=Curve, loc=<0 0 0>, rot=<0 0 0 1>}
 L_brow_Ot_Up_pose {type=Curve, loc=<0 0 0>, rot=<0 0 0 1>}
 L_brow_In_Dn_pose {type=Curve, loc=<0 0 0>, rot=<0 0 0 1>}
 L_brow_Ot_Dn_pose {type=Curve, loc=<0 0 0>, rot=<0 0 0 1>}
 R_brow_In_Up_pose {type=Curve, loc=<0 0 0>, rot=<0 0 0 1>}
 R_brow_Ot_Up_pose {type=Curve, loc=<0 0 0>, rot=<0 0 0 1>}
 R_brow_In_Dn_pose {type=Curve, loc=<0 0 0>, rot=<0 0 0 1>}
 R_brow_Ot_Dn_pose {type=Curve, loc=<0 0 0>, rot=<0 0 0 1>}
 upper_lip_up_pose {type=Curve, loc=<0 0 0>, rot=<0 0 0 1>}
 upper_lip_dn_pose {type=Curve, loc=<0 0 0>, rot=<0 0 0 1>}
 lower_lip_up_pose {type=Curve, loc=<0 0 0>, rot=<0 0 0 1>}
 lower_lip_dn_pose {type=Curve, loc=<0 0 0>, rot=<0 0 0 1>}
 L_smile_open_correction_pose {type=Curve, loc=<0 0 0>, rot=<0 0 0 1>}
 R_smile_open_correction_pose {type=Curve, loc=<0 0 0>, rot=<0 0 0 1>}
 tongue_left_pose {type=Curve, loc=<0 0 0>, rot=<0 0 0 1>}
 tongue_right_pose {type=Curve, loc=<0 0 0>, rot=<0 0 0 1>}
 base_pose {type=Curve, loc=<0 0 0>, rot=<0 0 0 1>}
 L_eye_look_up_pose {type=Curve, loc=<0 0 0>, rot=<0 0 0 1>}
 L_eye_look_left_pose {type=Curve, loc=<0 0 0>, rot=<0 0 0 1>}
 L_eye_look_down_pose {type=Curve, loc=<0 0 0>, rot=<0 0 0 1>}
 L_eye_look_right_pose {type=Curve, loc=<0 0 0>, rot=<0 0 0 1>}
 R_eye_look_left_pose {type=Curve, loc=<0 0 0>, rot=<0 0 0 1>}
 R_eye_look_right_pose {type=Curve, loc=<0 0 0>, rot=<0 0 0 1>}
 R_eye_look_up_pose {type=Curve, loc=<0 0 0>, rot=<0 0 0 1>}
 R_eye_look_down_pose {type=Curve, loc=<0 0 0>, rot=<0 0 0 1>}
 L_lower_lid_up_pose {type=Curve, loc=<0 0 0>, rot=<0 0 0 1>}
 R_lower_lid_up_pose {type=Curve, loc=<0 0 0>, rot=<0 0 0 1>}
 tongue_curl_tip_up_pose {type=Curve, loc=<0 0 0>, rot=<0 0 0 1>}
 tongue_curl_tip_down_pose {type=Curve, loc=<0 0 0>, rot=<0 0 0 1>}
 tongue_curl_mid_up_pose {type=Curve, loc=<0 0 0>, rot=<0 0 0 1>}
 tongue_curl_mid_down_pose {type=Curve, loc=<0 0 0>, rot=<0 0 0 1>}
 L_tongue_twist_pose {type=Curve, loc=<0 0 0>, rot=<0 0 0 1>}
 R_tongue_twist_pose {type=Curve, loc=<0 0 0>, rot=<0 0 0 1>}
 root/Root {type=Connector, loc=<0 0 0>, rot=<0 0 0 1>}
 root/Pelvis {type=Connector, loc=<0 0 0>, rot=<0 0 0 1>}
 root/Body Control Position {type=Connector, loc=<0 0 0>, rot=<0 0 0 1>}
 Quad_Spine_01/SpineRoot {type=Connector, loc=<0 0 0>, rot=<0 0 0 1>}
 Quad_Spine_01/Spine 3 {type=Connector, loc=<0 0 0>, rot=<0 0 0 1>}
 Quad_Spine_01/Spine 1 {type=Connector, loc=<0 0 0>, rot=<0 0 0 1>}
 Quad_Spine_01/parent {type=Connector, loc=<0 0 0>, rot=<0 0 0 1>}
 Quad_Spine_01/Spine 2 {type=Connector, loc=<0 0 0>, rot=<0 0 0 1>}
 Quad_Neck_01/Root {type=Connector, loc=<0 0 0>, rot=<0 0 0 1>}
 Quad_Neck_01/Head {type=Connector, loc=<0 0 0>, rot=<0 0 0 1>}
 Quad_Neck_01/NeckTip {type=Connector, loc=<0 0 0>, rot=<0 0 0 1>}
 Quad_Neck_01/Parent {type=Connector, loc=<0 0 0>, rot=<0 0 0 1>}
 Left Back Leg/Hip Socket {type=Connector, loc=<0 0 0>, rot=<0 0 0 1>}
 Left Back Leg/Knee {type=Connector, loc=<0 0 0>, rot=<0 0 0 1>}
 Left Back Leg/Ankle {type=Connector, loc=<0 0 0>, rot=<0 0 0 1>}
 Left Back Leg/Ball {type=Connector, loc=<0 0 0>, rot=<0 0 0 1>}
 Left Back Leg/Toe {type=Connector, loc=<0 0 0>, rot=<0 0 0 1>}
 Left Back Leg/Toe Tip {type=Connector, loc=<0 0 0>, rot=<0 0 0 1>}
 Left Back Leg/IK Rear Connector {type=Connector, loc=<0 0 0>, rot=<0 0 0 1>}
 Right Back Leg/Hip Socket {type=Connector, loc=<0 0 0>, rot=<0 0 0 1>}
 Right Back Leg/Knee {type=Connector, loc=<0 0 0>, rot=<0 0 0 1>}
 Right Back Leg/Ankle {type=Connector, loc=<0 0 0>, rot=<0 0 0 1>}
 Right Back Leg/Ball {type=Connector, loc=<0 0 0>, rot=<0 0 0 1>}
 Right Back Leg/Toe {type=Connector, loc=<0 0 0>, rot=<0 0 0 1>}
 Right Back Leg/Toe Tip {type=Connector, loc=<0 0 0>, rot=<0 0 0 1>}
 Right Back Leg/IK Rear Connector {type=Connector, loc=<0 0 0>, rot=<0 0 0 1>}
 Left Front Leg/Shoulder {type=Connector, loc=<0 0 0>, rot=<0 0 0 1>}
 Left Front Leg/Scapula Tip {type=Connector, loc=<0 0 0>, rot=<0 0 0 1>}
 Left Front Leg/Elbow {type=Connector, loc=<0 0 0>, rot=<0 0 0 1>}
 Left Front Leg/Wrist {type=Connector, loc=<0 0 0>, rot=<0 0 0 1>}
 Left Front Leg/Ball {type=Connector, loc=<0 0 0>, rot=<0 0 0 1>}
 Left Front Leg/Toe {type=Connector, loc=<0 0 0>, rot=<0 0 0 1>}
 Left Front Leg/Toe Tip {type=Connector, loc=<0 0 0>, rot=<0 0 0 1>}
 Left Front Leg/IK Front Connector {type=Connector, loc=<0 0 0>, rot=<0 0 0 1>}
 Right Front Leg/Shoulder {type=Connector, loc=<0 0 0>, rot=<0 0 0 1>}
 Right Front Leg/Scapula Tip {type=Connector, loc=<0 0 0>, rot=<0 0 0 1>}
 Right Front Leg/Elbow {type=Connector, loc=<0 0 0>, rot=<0 0 0 1>}
 Right Front Leg/Wrist {type=Connector, loc=<0 0 0>, rot=<0 0 0 1>}
 Right Front Leg/Ball {type=Connector, loc=<0 0 0>, rot=<0 0 0 1>}
 Right Front Leg/Toe {type=Connector, loc=<0 0 0>, rot=<0 0 0 1>}
 Right Front Leg/Toe Tip {type=Connector, loc=<0 0 0>, rot=<0 0 0 1>}
 Right Front Leg/IK Front Connector {type=Connector, loc=<0 0 0>, rot=<0 0 0 1>}
 Quad_Tail/HipRoot {type=Connector, loc=<0 0 0>, rot=<0 0 0 1>}
 Quad_Tail/TailTip {type=Connector, loc=<0 0 0>, rot=<0 0 0 1>}
 Quad_Tail/TailBase {type=Connector, loc=<0 0 0>, rot=<0 0 0 1>}
 LeftCape/Start {type=Connector, loc=<0 0 0>, rot=<0 0 0 1>}
 LeftCape/Parent {type=Connector, loc=<0 0 0>, rot=<0 0 0 1>}
 LeftCape/End {type=Connector, loc=<0 0 0>, rot=<0 0 0 1>}
 LeftCape/OrientSpace {type=Connector, loc=<0 0 0>, rot=<0 0 0 1>}
 CRM_Quad_Fk_v01/Start {type=Connector, loc=<0 0 0>, rot=<0 0 0 1>}
 CRM_Quad_Fk_v01/Parent {type=Connector, loc=<0 0 0>, rot=<0 0 0 1>}
 CRM_Quad_Fk_v01/End {type=Connector, loc=<0 0 0>, rot=<0 0 0 1>}
 CRM_Quad_Fk_v01/OrientSpace {type=Connector, loc=<0 0 0>, rot=<0 0 0 1>}
 CRM_Quad_Fk_v01_1/Start {type=Connector, loc=<0 0 0>, rot=<0 0 0 1>}
 CRM_Quad_Fk_v01_1/Parent {type=Connector, loc=<0 0 0>, rot=<0 0 0 1>}
 CRM_Quad_Fk_v01_1/End {type=Connector, loc=<0 0 0>, rot=<0 0 0 1>}
 CRM_Quad_Fk_v01_1/OrientSpace {type=Connector, loc=<0 0 0>, rot=<0 0 0 1>}
 Quad_Face/Head {type=Connector, loc=<0 0 0>, rot=<0 0 0 1>}
 Left Ear Root/Start {type=Connector, loc=<0 0 0>, rot=<0 0 0 1>}
 Left Ear Root/Parent {type=Connector, loc=<0 0 0>, rot=<0 0 0 1>}
 Left Ear Root/End {type=Connector, loc=<0 0 0>, rot=<0 0 0 1>}
 Left Ear Root/OrientSpace {type=Connector, loc=<0 0 0>, rot=<0 0 0 1>}
 Right Ear Root/Start {type=Connector, loc=<0 0 0>, rot=<0 0 0 1>}
 Right Ear Root/Parent {type=Connector, loc=<0 0 0>, rot=<0 0 0 1>}
 Right Ear Root/End {type=Connector, loc=<0 0 0>, rot=<0 0 0 1>}
 Right Ear Root/OrientSpace {type=Connector, loc=<0 0 0>, rot=<0 0 0 1>}
 Left Ear/Start {type=Connector, loc=<0 0 0>, rot=<0 0 0 1>}
 Left Ear/Parent {type=Connector, loc=<0 0 0>, rot=<0 0 0 1>}
 Left Ear/End {type=Connector, loc=<0 0 0>, rot=<0 0 0 1>}
 Left Ear/OrientSpace {type=Connector, loc=<0 0 0>, rot=<0 0 0 1>}
 Right Ear/Start {type=Connector, loc=<0 0 0>, rot=<0 0 0 1>}
 Right Ear/Parent {type=Connector, loc=<0 0 0>, rot=<0 0 0 1>}
 Right Ear/End {type=Connector, loc=<0 0 0>, rot=<0 0 0 1>}
 Right Ear/OrientSpace {type=Connector, loc=<0 0 0>, rot=<0 0 0 1>}
]
