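; Animation modifier for WhisperRemedy companion rigs that strips bone animation from the sequence it is applied to.
; On apply, it makes one RemoveBoneSelectiveAnimation call with face_attach as the named bone and both boolean options enabled.
; A hard-coded list of both ear bones and the left, middle and right dynamic cape master and target bones drives the selection, tying it to that skeleton's naming.
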(event EventOnApply
  (params (AnimationSequence :type Anim%20Sequence%20Object%20Reference))
  (AnimationBlueprintLibrary|Bones|RemoveBoneSelectiveAnimation AnimationSequence "face_attach" (Utilities|Array|MakeArray "ear_01_l" "ear_01_r" "dyn_cape_l_01_MSTR" "dyn_cape_mid_01_MSTR" "dyn_cape_r_01_MSTR" "dyn_cape_l_04_tgt_MSTR" "dyn_cape_mid_04_tgt_MSTR" "dyn_cape_r_04_tgt_MSTR" "dyn_cape_l_03_tgt_MSTR" "dyn_cape_mid_03_tgt_MSTR" "dyn_cape_r_03_tgt_MSTR" "dyn_cape_l_02_tgt_MSTR" "dyn_cape_mid_02_tgt_MSTR" "dyn_cape_r_02_tgt_MSTR") true true))
