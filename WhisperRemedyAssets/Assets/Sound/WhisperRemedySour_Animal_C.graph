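# Sound cue that plays the Whisper Remedy Sour animal sound, variant C, from a single wave.
# A Modulator node wraps the wave player, with the cue's overall volume boosted.
# It uses a dedicated attenuation asset and routes through a third-person ranged weapon fire sound class.
# The cue primes on load and overrides concurrency to cap simultaneous instances.
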SoundCue /WhisperRemedyAssets/Assets/Sound/WhisperRemedySour_Animal_C
root=Modulator_0
nodes [
 {id=Modulator_0, children [WavePlayer_0]}
 {id=WavePlayer_0, properties {soundWaveAssetPtr=&/IP_WhisperRemedySour_Animal_01C}}
]
properties {
 attenuationSettings=/WhisperRemedyAssets/Assets/Sound/Settings/IP_WhisperRemedySour_Animal_Att
 bOverrideConcurrency=true
 bPrimeOnLoad=true
 maxDistance=7000
 soundClassObject=&Fort_WeaponRanged_Fire_3P_PSM
 volumeMultiplier=2
 concurrencyOverrides {maxCount=8}
}
