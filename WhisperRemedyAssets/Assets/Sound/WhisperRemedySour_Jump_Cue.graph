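# Sound Cue for the WhisperRemedySour jump, layering an animal jump vocal with a whoosh delayed by a fixed short time.
# A Mixer combines the two branches, each with its own attenuation asset and a Modulator.
# Quality Level nodes give higher-index levels smaller random pools of two variations instead of four, with those pools excluded from branch culling.
# The cue primes on load, overrides concurrency and uses a third-person ranged weapon fire sound class.
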SoundCue /WhisperRemedyAssets/Assets/Sound/WhisperRemedySour_Jump_Cue
root=Mixer_0
nodes [
 {id=Attenuation_0, children [Delay_0], properties {attenuationSettings=/WhisperRemedyAssets/Assets/Sound/Settings/WhisperRemedySour_Jump_Att}}
 {
  id=Attenuation_1
  children [Modulator_0]
  properties {attenuationSettings=/WhisperRemedyAssets/Assets/Sound/Settings/IP_WhisperRemedySour_Animal_Att}
 }
 {id=Delay_0, children [Modulator_1], properties {delayMax=0.4, delayMin=0.4}}
 {id=Mixer_0, children [Attenuation_0, Attenuation_1], properties {inputVolume [1,1]}}
 {id=Modulator_0, children [QualityLevel_0]}
 {id=Modulator_1, children [QualityLevel_1]}
 {id=QualityLevel_0, children [Random_1,Random_1,Random_2,Random_2]}
 {id=QualityLevel_1, children [Random_0,Random_0,Random_3,Random_3]}
 {id=Random_0, children [WavePlayer_4,WavePlayer_5,WavePlayer_6,WavePlayer_7], properties {weights [1,1,1,1]}}
 {id=Random_1, children [WavePlayer_0,WavePlayer_1,WavePlayer_2,WavePlayer_3], properties {weights [1,1,1,1]}}
 {id=Random_2, children [WavePlayer_0,WavePlayer_3], properties {bShouldExcludeFromBranchCulling=true, weights [1,1]}}
 {id=Random_3, children [WavePlayer_4,WavePlayer_7], properties {bShouldExcludeFromBranchCulling=true, weights [1,1]}}
 {id=WavePlayer_0, properties {soundWaveAssetPtr=&/IP_WhisperRemedySour_Animal_Jump_01}}
 {id=WavePlayer_1, properties {soundWaveAssetPtr=&/IP_WhisperRemedySour_Animal_Jump_02}}
 {id=WavePlayer_2, properties {soundWaveAssetPtr=&/IP_WhisperRemedySour_Animal_Jump_03}}
 {id=WavePlayer_3, properties {soundWaveAssetPtr=&/IP_WhisperRemedySour_Animal_Jump_04}}
 {id=WavePlayer_4, properties {soundWaveAssetPtr=&/IP_WhisperRemedySour_Jump_Whoosh_01}}
 {id=WavePlayer_5, properties {soundWaveAssetPtr=&/IP_WhisperRemedySour_Jump_Whoosh_02}}
 {id=WavePlayer_6, properties {soundWaveAssetPtr=&/IP_WhisperRemedySour_Jump_Whoosh_03}}
 {id=WavePlayer_7, properties {soundWaveAssetPtr=&/IP_WhisperRemedySour_Jump_Whoosh_04}}
]
properties {
 bOverrideConcurrency=true
 bPrimeOnLoad=true
 maxDistance=9000
 soundClassObject=&Fort_WeaponRanged_Fire_3P_PSM
 volumeMultiplier=1
 concurrencyOverrides {maxCount=8}
}
