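# Sound cue for the WhisperRemedySour take-off, playing an exit sound layered with a boom, spatialized through a sonic-boom attenuation and the third-person ranged-weapon fire sound class.
# A Mixer combines two Modulator branches: one randomly picks from three boom waves, the other picks exit waves through a Quality Level node.
# That node gives two quality slots four random variations and the other two only two.
# Priming on load and a concurrency override are enabled.
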SoundCue /WhisperRemedyAssets/Assets/Sound/WhisperRemedySour_TakeOff_Cue
root=Mixer_0
nodes [
 {id=Mixer_0, children [Modulator_1,Modulator_0], properties {inputVolume [1,1]}}
 {id=Modulator_0, children [Random_1]}
 {id=Modulator_1, children [QualityLevel_1]}
 {id=QualityLevel_1, children [Random_0,Random_0,Random_3,Random_3]}
 {id=Random_0, children [WavePlayer_7, WavePlayer_8, WavePlayer_9, WavePlayer_10], properties {weights [1,1,1,1]}}
 {id=Random_1, children [WavePlayer_11, WavePlayer_12, WavePlayer_13], properties {weights [1,1,1]}}
 {id=Random_3, children [WavePlayer_7, WavePlayer_10], properties {bShouldExcludeFromBranchCulling=true, weights [1,1]}}
 {id=WavePlayer_10, properties {soundWaveAssetPtr=&/IP_WhisperRemedySour_Exit_04}}
 {id=WavePlayer_11, properties {soundWaveAssetPtr=&/IP_WhisperRemedySour_Exit_Boom_01}}
 {id=WavePlayer_12, properties {soundWaveAssetPtr=&/IP_WhisperRemedySour_Exit_Boom_02}}
 {id=WavePlayer_13, properties {soundWaveAssetPtr=&/IP_WhisperRemedySour_Exit_Boom_03}}
 {id=WavePlayer_7, properties {soundWaveAssetPtr=&/IP_WhisperRemedySour_Exit_01}}
 {id=WavePlayer_8, properties {soundWaveAssetPtr=&/IP_WhisperRemedySour_Exit_02}}
 {id=WavePlayer_9, properties {soundWaveAssetPtr=&/IP_WhisperRemedySour_Exit_03}}
]
properties {
 attenuationSettings=/WhisperRemedyAssets/Assets/Sound/Settings/WhisperRemedySour_SonicBoom_Att
 bOverrideConcurrency=true
 bPrimeOnLoad=true
 maxDistance=9000
 soundClassObject=&Fort_WeaponRanged_Fire_3P_PSM
 volumeMultiplier=1
 concurrencyOverrides {maxCount=8}
}
